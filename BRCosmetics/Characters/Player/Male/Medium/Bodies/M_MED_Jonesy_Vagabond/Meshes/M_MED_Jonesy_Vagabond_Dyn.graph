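# Physics asset for secondary motion on the Jonesy Vagabond character's dangling accessories: left and right chest pouches, a straw, and chest and back zippers.
# Capsule bodies form short chains from MSTR root bones, with a two-segment flap on the left pouch and the straw on the right.
# Child joints set narrow twist and wider swing limits.
# Only the zipper roots are constrained to the spine_05 capsule, with free angular motion.
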PhysicsAsset /BRCosmetics/Characters/Player/Male/Medium/Bodies/M_MED_Jonesy_Vagabond/Meshes/M_MED_Jonesy_Vagabond_Dyn
bodies {
 dyn_pouch_chest_l [{shapeType=Capsule, length=7.0625, radius=2.875, center=<2.0114 0 2.8195>, rotation=<rot 90 0 0>}]
 dyn_pouch_chest_l_MSTR [{shapeType=Capsule, length=0.505, radius=0.505}]
 dyn_pouch_chest_l_flap [{shapeType=Capsule, length=5.1281, radius=0.5, center=<3.1875 0 0>, rotation=<rot 90 0 0>}]
 dyn_pouch_chest_l_flap_02 [{shapeType=Capsule, length=4.0344, radius=0.5, center=<2.1454 0 0>, rotation=<rot 90 0 0>}]
 dyn_pouch_chest_r [{shapeType=Capsule, length=7.0625, radius=2.875, center=<2.0114 0 2.8195>, rotation=<rot 90 0 0>}]
 dyn_pouch_chest_r_MSTR [{shapeType=Capsule, length=0.505, radius=0.505}]
 dyn_pouch_chest_r_straw [{shapeType=Capsule, length=2.0031, radius=0.5, center=<1.6421 0 0>, rotation=<rot 90 0 0>}]
 dyn_zipper_back [{shapeType=Capsule, length=2.0031, radius=0.5, center=<1.6421 0 0>, rotation=<rot 90 0 0>}]
 dyn_zipper_back_MSTR [{shapeType=Capsule, length=0.505, radius=0.505}]
 dyn_zipper_chest [{shapeType=Capsule, length=2.0031, radius=0.5, center=<1.6421 0 0>, rotation=<rot 90 0 0>}]
 dyn_zipper_chest_MSTR [{shapeType=Capsule, length=0.505, radius=0.505}]
 spine_05 [{shapeType=Capsule, length=14.0557, radius=5.5228, center=<-0.428 0.7046 0>}]
}
constraints [
 {
  bone1Name=dyn_zipper_chest_MSTR
  bone2Name=spine_05
  swing1LimitDegrees=45
  swing1Motion=Free
  swing2LimitDegrees=45
  swing2Motion=Free
  twistLimitDegrees=45
  twistMotion=Free
 }
 {bone1Name=dyn_zipper_chest, bone2Name=dyn_zipper_chest_MSTR, swing1LimitDegrees=40, swing2LimitDegrees=20, twistLimitDegrees=5}
 {
  bone1Name=dyn_zipper_back_MSTR
  bone2Name=spine_05
  swing1LimitDegrees=45
  swing1Motion=Free
  swing2LimitDegrees=45
  swing2Motion=Free
  twistLimitDegrees=45
  twistMotion=Free
 }
 {bone1Name=dyn_zipper_back, bone2Name=dyn_zipper_back_MSTR, swing1LimitDegrees=40, swing2LimitDegrees=20, twistLimitDegrees=5}
 {bone1Name=dyn_pouch_chest_r, bone2Name=dyn_pouch_chest_r_MSTR, swing1LimitDegrees=45, swing2LimitDegrees=45, twistLimitDegrees=5}
 {bone1Name=dyn_pouch_chest_r_straw, bone2Name=dyn_pouch_chest_r, swing1LimitDegrees=45, swing2LimitDegrees=45, twistLimitDegrees=5}
 {bone1Name=dyn_pouch_chest_l, bone2Name=dyn_pouch_chest_l_MSTR, swing1LimitDegrees=45, swing2LimitDegrees=45, twistLimitDegrees=5}
 {bone1Name=dyn_pouch_chest_l_flap, bone2Name=dyn_pouch_chest_l, swing1LimitDegrees=20, swing2LimitDegrees=20, twistLimitDegrees=5}
 {bone1Name=dyn_pouch_chest_l_flap_02, bone2Name=dyn_pouch_chest_l_flap, swing1LimitDegrees=20, swing2LimitDegrees=20, twistLimitDegrees=5}
]
