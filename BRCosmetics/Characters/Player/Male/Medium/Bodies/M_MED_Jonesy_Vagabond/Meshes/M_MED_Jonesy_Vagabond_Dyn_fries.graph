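# Physics asset for the dangling fries cup on the Jonesy Vagabond body cosmetic, giving the cup and its four fries secondary motion.
# Small capsule bodies cover a master bone, the cup and each fry.
# The cup rotates freely on swing and twist relative to the master, while each fry is constrained directly to the cup with tight angular limits, forming a shallow hub-and-spoke rig.
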PhysicsAsset /BRCosmetics/Characters/Player/Male/Medium/Bodies/M_MED_Jonesy_Vagabond/Meshes/M_MED_Jonesy_Vagabond_Dyn_fries
bodies {
 dyn_frie_01 [{shapeType=Capsule, length=2.0031, radius=0.5, center=<1.6421 0 0>, rotation=<rot 90 0 0>}]
 dyn_frie_02 [{shapeType=Capsule, length=2.0031, radius=0.5, center=<1.6421 0 0>, rotation=<rot 90 0 0>}]
 dyn_frie_03 [{shapeType=Capsule, length=2.0031, radius=0.5, center=<1.6421 0 0>, rotation=<rot 90 0 0>}]
 dyn_frie_04 [{shapeType=Capsule, length=2.0031, radius=0.5, center=<1.6421 0 0>, rotation=<rot 90 0 0>}]
 dyn_fries_cup [{shapeType=Capsule, length=0.505, radius=0.505}]
 dyn_fries_cup_MSTR [{shapeType=Capsule, length=0.505, radius=0.505}]
}
constraints [
 {
  bone1Name=dyn_fries_cup
  bone2Name=dyn_fries_cup_MSTR
  swing1LimitDegrees=45
  swing1Motion=Free
  swing2LimitDegrees=45
  swing2Motion=Free
  twistLimitDegrees=45
  twistMotion=Free
 }
 {bone1Name=dyn_frie_01, bone2Name=dyn_fries_cup, swing1LimitDegrees=5, swing2LimitDegrees=5, twistLimitDegrees=5}
 {bone1Name=dyn_frie_02, bone2Name=dyn_fries_cup, swing1LimitDegrees=5, swing2LimitDegrees=5, twistLimitDegrees=5}
 {bone1Name=dyn_frie_03, bone2Name=dyn_fries_cup, swing1LimitDegrees=5, swing2LimitDegrees=5, twistLimitDegrees=5}
 {bone1Name=dyn_frie_04, bone2Name=dyn_fries_cup, swing1LimitDegrees=5, swing2LimitDegrees=5, twistLimitDegrees=5}
]
