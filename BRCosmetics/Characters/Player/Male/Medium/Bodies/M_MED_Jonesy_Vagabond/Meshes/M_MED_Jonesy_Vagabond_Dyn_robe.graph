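# Physics asset for dynamic secondary motion of the Jonesy Vagabond male character's robe, fanny-pack zipper and pants string.
# Seven five-segment capsule coat chains hang from spine_02, with wide root swing and tighter swing and twist down each chain.
# Angularly free cross-constraints tie some neighboring side, front and back chains together.
# Zipper and string hang from pelvis master bodies, alongside capsules for pelvis, spine, thighs and calves.
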PhysicsAsset /BRCosmetics/Characters/Player/Male/Medium/Bodies/M_MED_Jonesy_Vagabond/Meshes/M_MED_Jonesy_Vagabond_Dyn_robe
bodies {
 calf_l [{shapeType=Capsule, length=46.5884, radius=8.4266, center=<-26.3168 -2.7608 0.0003>, rotation=<rot 90 0 0>}]
 calf_r [{shapeType=Capsule, length=46.5884, radius=8.4266, center=<27.0703 2.7608 -0.0004>, rotation=<rot 90 0 0>}]
 dyn_coat_bk_01 [{shapeType=Capsule, length=9.2812, radius=10.25, center=<4.2067 0 9.5255>, rotation=<rot 90 0 0>}]
 dyn_coat_bk_02 [{shapeType=Capsule, length=13.3438, radius=10.0312, center=<8.003 0 10>, rotation=<rot 0 0 90>}]
 dyn_coat_bk_03 [{shapeType=Capsule, length=10.4688, radius=10.5625, center=<6.6842 0 10>, rotation=<rot 0 0 90>}]
 dyn_coat_bk_04 [{shapeType=Capsule, length=24, radius=10.75, shapeName=dyn_coat_bk_04_capsule, center=<5.8953 0 8>, rotation=<rot 0 0 89.9998>}]
 dyn_coat_bk_05 [{shapeType=Capsule, length=24, radius=10.75, shapeName=dyn_coat_bk_05_capsule, center=<5.8953 0 8>, rotation=<rot 0 0 89.9998>}]
 dyn_coat_bk_l_01 [{shapeType=Capsule, length=6.25, radius=10, center=<3.3961 0.0001 9.5255>, rotation=<rot 90 0 0>}]
 dyn_coat_bk_l_02 [{shapeType=Capsule, length=14.7188, radius=9.5, center=<6.5 0 9.5>, rotation=<rot 85 -89.9985 0.0014>}]
 dyn_coat_bk_l_03 [{shapeType=Capsule, length=14.7188, radius=9.5, center=<6.5 0 9.5>, rotation=<rot 85 -89.9985 0.0014>}]
 dyn_coat_bk_l_04 [{shapeType=Capsule, length=14.2188, radius=10, center=<6.5 0 9.5>, rotation=<rot 85 -89.9975 0.0023>}]
 dyn_coat_bk_l_05 [{shapeType=Capsule, length=14.2188, radius=10, center=<6.5 0 9.5>, rotation=<rot 85 -89.9976 0.0023>}]
 dyn_coat_bk_r_01 [{shapeType=Capsule, length=11.4062, radius=8.125, center=<4.2067 3.3441 8.5088>, rotation=<rot 90 0 90>}]
 dyn_coat_bk_r_02 [{shapeType=Capsule, length=16.0938, radius=8.125, center=<6.5 0 8.5>, rotation=<rot 74.9999 89.9988 179.999>}]
 dyn_coat_bk_r_03 [{shapeType=Capsule, length=14.5625, radius=8.125, center=<6.5 0 8.5>, rotation=<rot 75 89.9988 179.9991>}]
 dyn_coat_bk_r_04 [{shapeType=Capsule, length=13.25, radius=8.75, center=<6.5 0 8.5>, rotation=<rot 75 89.9988 179.9991>}]
 dyn_coat_bk_r_05 [{shapeType=Capsule, length=13.25, radius=8.75, center=<6.5 0 8.5>, rotation=<rot 75 89.9985 179.9988>}]
 dyn_coat_fr_l_01 [{shapeType=Capsule, length=12.4688, radius=10, center=<5.7664 0 9.688>, rotation=<rot 86.749 179.9995 179.9995>}]
 dyn_coat_fr_l_02 [{shapeType=Capsule, length=6.25, radius=10, center=<3.3961 0.0001 9.5255>, rotation=<rot 90 0 0>}]
 dyn_coat_fr_l_03 [{shapeType=Capsule, length=6, radius=8.25, shapeName=dyn_coat_fr_l_04_capsule, center=<6 0 5>, rotation=<rot 0.0001 0.0002 105>}]
 dyn_coat_fr_l_04 [{shapeType=Capsule, length=7.125, radius=7.125, shapeName=dyn_coat_fr_l_04_capsule, center=<6 0 4>, rotation=<rot 0.0001 0.0002 105>}]
 dyn_coat_fr_l_05 [{shapeType=Capsule, length=7.125, radius=7.125, shapeName=dyn_coat_fr_l_04_capsule, center=<6 0 4>, rotation=<rot 0.0001 0.0002 105>}]
 dyn_coat_fr_r_01 [{shapeType=Capsule, length=12.4688, radius=10, center=<5.7664 0 9.688>, rotation=<rot 86.749 179.9995 179.9995>}]
 dyn_coat_fr_r_02 [{shapeType=Capsule, length=6.25, radius=10, center=<3.3961 0.0001 9.5255>, rotation=<rot 90 0 0>}]
 dyn_coat_fr_r_03 [{shapeType=Capsule, length=7.75, radius=6.5, shapeName=dyn_coat_fr_l_04_capsule, center=<6 5 5>, rotation=<rot 0 0 90>}]
 dyn_coat_fr_r_04 [{shapeType=Capsule, length=7.75, radius=6.5, shapeName=dyn_coat_fr_l_04_capsule, center=<6 5 5>, rotation=<rot 0 0 90>}]
 dyn_coat_fr_r_05 [{shapeType=Capsule, length=7.75, radius=6.5, shapeName=dyn_coat_fr_l_04_capsule, center=<6 5 5>, rotation=<rot 0 0 90>}]
 dyn_coat_l_01 [{shapeType=Capsule, length=9.375, radius=10, center=<4.017 -3.6758 8.2178>, rotation=<rot 65.902 -89.9996 -89.9996>}]
 dyn_coat_l_02 [{shapeType=Capsule, length=2.375, radius=9.5, center=<3.8636 0 9>, rotation=<rot 0 0 90>}]
 dyn_coat_l_03 [{shapeType=Capsule, length=12.6562, radius=9.5, center=<5.8161 0 9>, rotation=<rot 0 0 90>}]
 dyn_coat_l_04 [{shapeType=Capsule, length=12.3125, radius=10, center=<6.2371 0 9>, rotation=<rot 0 0 90>}]
 dyn_coat_l_05 [{shapeType=Capsule, length=12.8125, radius=10, center=<6.4106 0 9>, rotation=<rot 0 0 90>}]
 dyn_coat_r_01 [{shapeType=Capsule, length=9.2812, radius=10.25, center=<4.2067 0 9.5255>, rotation=<rot 90 0 0>}]
 dyn_coat_r_02 [{shapeType=Capsule, length=1.5, radius=10.25, center=<2.7235 0 9.52>, rotation=<rot 90 0 0>}]
 dyn_coat_r_03 [{shapeType=Capsule, length=11.9062, radius=10.25, center=<6.1906 0 9.52>, rotation=<rot 90 0 0>}]
 dyn_coat_r_04 [{shapeType=Capsule, length=11.9062, radius=10.25, center=<5.9775 0 9.52>, rotation=<rot 90 0 0>}]
 dyn_coat_r_05 [{shapeType=Capsule, length=12.3125, radius=10.25, center=<6.6326 0 9.52>, rotation=<rot 90 0 0>}]
 dyn_pants_string_01 [{shapeType=Capsule, length=3.2844, radius=0.5, center=<2.0266 0 0>, rotation=<rot 90 0 0>}]
 dyn_pants_string_02 [{shapeType=Capsule, length=3.5344, radius=0.875, center=<3.0915 0 0>, rotation=<rot 90 0 0>}]
 dyn_pants_string_03 [{shapeType=Capsule, length=3.5344, radius=0.875, center=<3.0915 0 0>, rotation=<rot 90 0 0>}]
 dyn_pants_string_MSTR [{shapeType=Capsule, length=0.505, radius=0.505}]
 dyn_zipper_fanny [{shapeType=Capsule, length=2.0031, radius=0.5, center=<1.6421 0 0>, rotation=<rot 90 0 0>}]
 dyn_zipper_fanny_MSTR [{shapeType=Capsule, length=0.505, radius=0.505}]
 pelvis [{shapeType=Capsule, length=8.3125, radius=14.1875, center=<-0.8507 1.5914 0>}]
 spine_01 [{shapeType=Capsule, length=5.9251, radius=14.1478, center=<-2.4193 1.8007 -0.4237>}]
 spine_02 [{shapeType=Capsule, length=4.5346, radius=13.2728, center=<-2.4193 3.2691 0.8186>}]
 thigh_l [{shapeType=Capsule, length=28.6729, radius=11.6959, center=<-19.3517 1.3419 1.5689>, rotation=<rot 44.9573 -90.1174 -90.7771>}]
 thigh_r [
  {shapeType=Capsule, length=13.1416, radius=5.8834, center=<10.3787 -4.4616 6.3091>, rotation=<rot 82.6955 -110.6022 -119.6249>}
  {shapeType=Capsule, length=29.2666, radius=11.1021, center=<19.3461 -0.2395 -1.8601>, rotation=<rot 44.9573 -90.1174 -90.7771>}
 ]
}
constraints [
 {bone1Name=dyn_coat_bk_l_01, bone2Name=spine_02, swing1LimitDegrees=45, swing2LimitDegrees=45, twistLimitDegrees=5}
 {bone1Name=dyn_coat_bk_l_02, bone2Name=dyn_coat_bk_l_01, swing1LimitDegrees=5, swing2LimitDegrees=45, twistLimitDegrees=2}
 {bone1Name=dyn_coat_bk_l_03, bone2Name=dyn_coat_bk_l_02, swing1LimitDegrees=5, swing2LimitDegrees=45, twistLimitDegrees=2}
 {bone1Name=dyn_coat_bk_l_04, bone2Name=dyn_coat_bk_l_03, swing1LimitDegrees=5, swing2LimitDegrees=45, twistLimitDegrees=2}
 {bone1Name=dyn_coat_bk_l_05, bone2Name=dyn_coat_bk_l_04, swing1LimitDegrees=5, swing2LimitDegrees=45, twistLimitDegrees=2}
 {bone1Name=dyn_coat_l_01, bone2Name=spine_02, swing1LimitDegrees=45, swing2LimitDegrees=45, twistLimitDegrees=5}
 {bone1Name=dyn_coat_l_02, bone2Name=dyn_coat_l_01, swing1LimitDegrees=5, swing2LimitDegrees=45, twistLimitDegrees=2}
 {bone1Name=dyn_coat_l_03, bone2Name=dyn_coat_l_02, swing1LimitDegrees=5, swing2LimitDegrees=45, twistLimitDegrees=2}
 {bone1Name=dyn_coat_l_04, bone2Name=dyn_coat_l_03, swing1LimitDegrees=5, swing2LimitDegrees=45, twistLimitDegrees=2}
 {bone1Name=dyn_coat_l_05, bone2Name=dyn_coat_l_04, swing1LimitDegrees=5, swing2LimitDegrees=45, twistLimitDegrees=2}
 {bone1Name=dyn_coat_fr_l_01, bone2Name=spine_02, swing1LimitDegrees=45, swing2LimitDegrees=45, twistLimitDegrees=5}
 {bone1Name=dyn_coat_fr_l_02, bone2Name=dyn_coat_fr_l_01, swing1LimitDegrees=5, swing2LimitDegrees=45, twistLimitDegrees=2}
 {bone1Name=dyn_coat_fr_l_03, bone2Name=dyn_coat_fr_l_02, swing1LimitDegrees=5, swing2LimitDegrees=45, twistLimitDegrees=2}
 {bone1Name=dyn_coat_fr_l_04, bone2Name=dyn_coat_fr_l_03, swing1LimitDegrees=5, swing2LimitDegrees=45, twistLimitDegrees=2}
 {bone1Name=dyn_coat_fr_l_05, bone2Name=dyn_coat_fr_l_04, swing1LimitDegrees=5, swing2LimitDegrees=45, twistLimitDegrees=2}
 {bone1Name=dyn_coat_fr_r_01, bone2Name=spine_02, swing1LimitDegrees=45, swing2LimitDegrees=45, twistLimitDegrees=5}
 {bone1Name=dyn_coat_fr_r_02, bone2Name=dyn_coat_fr_r_01, swing1LimitDegrees=5, swing2LimitDegrees=45, twistLimitDegrees=2}
 {bone1Name=dyn_coat_fr_r_03, bone2Name=dyn_coat_fr_r_02, swing1LimitDegrees=5, swing2LimitDegrees=45, twistLimitDegrees=2}
 {bone1Name=dyn_coat_fr_r_04, bone2Name=dyn_coat_fr_r_03, swing1LimitDegrees=5, swing2LimitDegrees=45, twistLimitDegrees=2}
 {bone1Name=dyn_coat_fr_r_05, bone2Name=dyn_coat_fr_r_04, swing1LimitDegrees=5, swing2LimitDegrees=45, twistLimitDegrees=2}
 {bone1Name=dyn_coat_r_01, bone2Name=spine_02, swing1LimitDegrees=45, swing2LimitDegrees=45, twistLimitDegrees=5}
 {bone1Name=dyn_coat_r_02, bone2Name=dyn_coat_r_01, swing1LimitDegrees=5, swing2LimitDegrees=45, twistLimitDegrees=2}
 {bone1Name=dyn_coat_r_03, bone2Name=dyn_coat_r_02, swing1LimitDegrees=5, swing2LimitDegrees=45, twistLimitDegrees=2}
 {bone1Name=dyn_coat_r_04, bone2Name=dyn_coat_r_03, swing1LimitDegrees=5, swing2LimitDegrees=45, twistLimitDegrees=2}
 {bone1Name=dyn_coat_r_05, bone2Name=dyn_coat_r_04, swing1LimitDegrees=5, swing2LimitDegrees=45, twistLimitDegrees=2}
 {bone1Name=dyn_coat_bk_01, bone2Name=spine_02, swing1LimitDegrees=45, swing2LimitDegrees=45, twistLimitDegrees=5}
 {bone1Name=dyn_coat_bk_02, bone2Name=dyn_coat_bk_01, swing1LimitDegrees=5, swing2LimitDegrees=45, twistLimitDegrees=2}
 {bone1Name=dyn_coat_bk_03, bone2Name=dyn_coat_bk_02, swing1LimitDegrees=5, swing2LimitDegrees=45, twistLimitDegrees=2}
 {bone1Name=dyn_coat_bk_r_01, bone2Name=spine_02, swing1LimitDegrees=45, swing2LimitDegrees=45, twistLimitDegrees=5}
 {bone1Name=dyn_coat_bk_r_02, bone2Name=dyn_coat_bk_r_01, swing1LimitDegrees=5, swing2LimitDegrees=45, twistLimitDegrees=2}
 {bone1Name=dyn_coat_bk_r_03, bone2Name=dyn_coat_bk_r_02, swing1LimitDegrees=5, swing2LimitDegrees=45, twistLimitDegrees=2}
 {bone1Name=dyn_coat_bk_r_04, bone2Name=dyn_coat_bk_r_03, swing1LimitDegrees=5, swing2LimitDegrees=45, twistLimitDegrees=2}
 {bone1Name=dyn_coat_bk_r_05, bone2Name=dyn_coat_bk_r_04, swing1LimitDegrees=5, swing2LimitDegrees=45, twistLimitDegrees=2}
 {
  bone1Name=spine_01
  bone2Name=pelvis
  swing1LimitDegrees=45
  swing1Motion=Free
  swing2LimitDegrees=45
  swing2Motion=Free
  twistLimitDegrees=45
  twistMotion=Free
 }
 {
  bone1Name=spine_02
  bone2Name=spine_01
  swing1LimitDegrees=45
  swing1Motion=Free
  swing2LimitDegrees=45
  swing2Motion=Free
  twistLimitDegrees=45
  twistMotion=Free
 }
 {
  bone1Name=thigh_l
  bone2Name=pelvis
  swing1LimitDegrees=45
  swing1Motion=Free
  swing2LimitDegrees=45
  swing2Motion=Free
  twistLimitDegrees=45
  twistMotion=Free
 }
 {
  bone1Name=calf_l
  bone2Name=thigh_l
  swing1LimitDegrees=45
  swing1Motion=Free
  swing2LimitDegrees=45
  swing2Motion=Free
  twistLimitDegrees=45
  twistMotion=Free
 }
 {
  bone1Name=thigh_r
  bone2Name=pelvis
  swing1LimitDegrees=45
  swing1Motion=Free
  swing2LimitDegrees=45
  swing2Motion=Free
  twistLimitDegrees=45
  twistMotion=Free
 }
 {
  bone1Name=calf_r
  bone2Name=thigh_r
  swing1LimitDegrees=45
  swing1Motion=Free
  swing2LimitDegrees=45
  swing2Motion=Free
  twistLimitDegrees=45
  twistMotion=Free
 }
 {
  bone1Name=dyn_coat_l_03
  bone2Name=dyn_coat_fr_l_03
  swing1LimitDegrees=45
  swing1Motion=Free
  swing2LimitDegrees=45
  swing2Motion=Free
  twistLimitDegrees=45
  twistMotion=Free
 }
 {
  bone1Name=dyn_coat_l_05
  bone2Name=dyn_coat_fr_l_05
  swing1LimitDegrees=45
  swing1Motion=Free
  swing2LimitDegrees=45
  swing2Motion=Free
  twistLimitDegrees=45
  twistMotion=Free
 }
 {
  bone1Name=dyn_coat_bk_l_03
  bone2Name=dyn_coat_l_03
  swing1LimitDegrees=45
  swing1Motion=Free
  swing2LimitDegrees=45
  swing2Motion=Free
  twistLimitDegrees=45
  twistMotion=Free
 }
 {
  bone1Name=dyn_coat_r_03
  bone2Name=dyn_coat_fr_r_03
  swing1LimitDegrees=45
  swing1Motion=Free
  swing2LimitDegrees=45
  swing2Motion=Free
  twistLimitDegrees=45
  twistMotion=Free
 }
 {
  bone1Name=dyn_coat_r_05
  bone2Name=dyn_coat_fr_r_05
  swing1LimitDegrees=45
  swing1Motion=Free
  swing2LimitDegrees=45
  swing2Motion=Free
  twistLimitDegrees=45
  twistMotion=Free
 }
 {
  bone1Name=dyn_coat_bk_r_03
  bone2Name=dyn_coat_r_03
  swing1LimitDegrees=45
  swing1Motion=Free
  swing2LimitDegrees=45
  swing2Motion=Free
  twistLimitDegrees=45
  twistMotion=Free
 }
 {bone1Name=dyn_coat_bk_04, bone2Name=dyn_coat_bk_03, swing1LimitDegrees=5, swing2LimitDegrees=45, twistLimitDegrees=2}
 {bone1Name=dyn_coat_bk_05, bone2Name=dyn_coat_bk_04, swing1LimitDegrees=5, swing2LimitDegrees=45, twistLimitDegrees=2}
 {
  bone1Name=dyn_zipper_fanny_MSTR
  bone2Name=pelvis
  swing1LimitDegrees=45
  swing1Motion=Free
  swing2LimitDegrees=45
  swing2Motion=Free
  twistLimitDegrees=45
  twistMotion=Free
 }
 {bone1Name=dyn_zipper_fanny, bone2Name=dyn_zipper_fanny_MSTR, swing1LimitDegrees=40, swing2LimitDegrees=20, twistLimitDegrees=5}
 {
  bone1Name=dyn_pants_string_MSTR
  bone2Name=pelvis
  swing1LimitDegrees=45
  swing1Motion=Free
  swing2LimitDegrees=45
  swing2Motion=Free
  twistLimitDegrees=45
  twistMotion=Free
 }
 {bone1Name=dyn_pants_string_01, bone2Name=dyn_pants_string_MSTR, swing1LimitDegrees=40, swing2LimitDegrees=40, twistLimitDegrees=5}
 {bone1Name=dyn_pants_string_02, bone2Name=dyn_pants_string_01, swing1LimitDegrees=40, swing2LimitDegrees=40, twistLimitDegrees=5}
 {bone1Name=dyn_pants_string_03, bone2Name=dyn_pants_string_01, swing1LimitDegrees=40, swing2LimitDegrees=40, twistLimitDegrees=5}
]
