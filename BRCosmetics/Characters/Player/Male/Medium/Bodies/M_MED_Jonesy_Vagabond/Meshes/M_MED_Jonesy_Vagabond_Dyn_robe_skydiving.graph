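# Physics asset for the skydiving dynamics of the Jonesy Vagabond robe, plus its pants string and fanny zipper.
# Seven five-segment coat chains hang from master bones on spine_02 with tight angular limits, while the string and zipper hang from the pelvis.
# Angular-free cross constraints tie some neighboring chains at segments 03 and 05 to link the cloth panels.
# Capsules on pelvis, lower spine, thighs and calves provide the body shapes.
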PhysicsAsset /BRCosmetics/Characters/Player/Male/Medium/Bodies/M_MED_Jonesy_Vagabond/Meshes/M_MED_Jonesy_Vagabond_Dyn_robe_skydiving
bodies {
 calf_l [{shapeType=Capsule, length=45.3384, radius=9.6766, center=<-26.3168 -2.7608 0.0003>, rotation=<rot 90 0 0>}]
 calf_r [{shapeType=Capsule, length=45.3384, radius=9.6766, center=<27.0703 2.7608 -0.0004>, rotation=<rot 90 0 0>}]
 dyn_coat_bk_01 [{shapeType=Capsule, length=21.125, radius=10.5, center=<6 0 9>, rotation=<rot 0 0 90>}]
 dyn_coat_bk_02 [{shapeType=Capsule, length=16.75, radius=10.5, center=<6 0 9>, rotation=<rot 0 0 90>}]
 dyn_coat_bk_03 [{shapeType=Capsule, length=28, radius=10.5, center=<6 0 9>, rotation=<rot 0 0 90>}]
 dyn_coat_bk_04 [{shapeType=Capsule, length=31.5, radius=10.5, shapeName=dyn_coat_bk_04_capsule, center=<6 0 9>, rotation=<rot 0 0 90>}]
 dyn_coat_bk_05 [{shapeType=Capsule, length=30.75, radius=11.25, shapeName=dyn_coat_bk_05_capsule, center=<9 0 9>, rotation=<rot 0 0 90>}]
 dyn_coat_bk_MSTR [{shapeType=Capsule, length=0.505, radius=0.505}]
 dyn_coat_bk_l_01 [{shapeType=Capsule, length=21.125, radius=10.5, center=<6 0 9>, rotation=<rot -0.0003 0.0003 75>}]
 dyn_coat_bk_l_02 [{shapeType=Capsule, length=21.125, radius=10.5, center=<6 0 9>, rotation=<rot -0.0003 0.0003 74.9999>}]
 dyn_coat_bk_l_03 [{shapeType=Capsule, length=22.375, radius=10.5, center=<6 0 9>, rotation=<rot -0.0001 0.0001 80>}]
 dyn_coat_bk_l_04 [{shapeType=Capsule, length=22.375, radius=10.5, center=<6 0 9>, rotation=<rot -0.0001 0.0001 80>}]
 dyn_coat_bk_l_05 [{shapeType=Capsule, length=21.625, radius=11.25, center=<9 0 9>, rotation=<rot -0.0001 0.0001 80>}]
 dyn_coat_bk_l_MSTR [{shapeType=Capsule, length=0.505, radius=0.505}]
 dyn_coat_bk_r_01 [{shapeType=Capsule, length=16.25, radius=10.5, center=<6 0 9>, rotation=<rot -0.0003 -0.0001 100>}]
 dyn_coat_bk_r_02 [{shapeType=Capsule, length=16.25, radius=10.5, center=<6 0 9>, rotation=<rot -0.0003 -0.0001 100>}]
 dyn_coat_bk_r_03 [{shapeType=Capsule, length=21.125, radius=10.5, center=<6 0 9>, rotation=<rot -0.0002 -0.0002 100>}]
 dyn_coat_bk_r_04 [{shapeType=Capsule, length=21.125, radius=10.5, center=<6 0 9>, rotation=<rot -0.0002 -0.0002 100>}]
 dyn_coat_bk_r_05 [{shapeType=Capsule, length=20.375, radius=11.25, shapeName=dyn_coat_bk_r_05_capsule, center=<9 0 9>, rotation=<rot -0.0002 -0.0002 100>}]
 dyn_coat_bk_r_MSTR [{shapeType=Capsule, length=0.505, radius=0.505}]
 dyn_coat_fr_l_01 [{shapeType=Capsule, length=12.4688, radius=10, center=<5.7664 0 9.688>, rotation=<rot 86.749 179.9995 179.9995>}]
 dyn_coat_fr_l_02 [{shapeType=Capsule, length=6.25, radius=10, center=<3.3961 0.0001 9.5255>, rotation=<rot 90 0 0>}]
 dyn_coat_fr_l_03 [{shapeType=Capsule, length=16.75, radius=10.5, shapeName=dyn_coat_fr_l_04_capsule, center=<6 0 9>, rotation=<rot 0 0 90>}]
 dyn_coat_fr_l_04 [{shapeType=Capsule, length=16.75, radius=10.5, shapeName=dyn_coat_fr_l_04_capsule, center=<6 0 9>, rotation=<rot 0 0 90>}]
 dyn_coat_fr_l_05 [{shapeType=Capsule, length=16.75, radius=10.5, shapeName=dyn_coat_fr_l_05_capsule, center=<9 0 9>, rotation=<rot 0 0 90>}]
 dyn_coat_fr_l_MSTR [{shapeType=Capsule, length=0.505, radius=0.505}]
 dyn_coat_fr_r_01 [{shapeType=Capsule, length=12.4688, radius=10, center=<5.7664 0 9.688>, rotation=<rot 86.749 179.9995 179.9995>}]
 dyn_coat_fr_r_02 [{shapeType=Capsule, length=6.25, radius=10, center=<3.3961 0.0001 9.5255>, rotation=<rot 90 0 0>}]
 dyn_coat_fr_r_03 [{shapeType=Capsule, length=16.75, radius=10.5, shapeName=dyn_coat_fr_r_03_capsule, center=<6 0 9>, rotation=<rot 0 0 90>}]
 dyn_coat_fr_r_04 [{shapeType=Capsule, length=16.75, radius=10.5, shapeName=dyn_coat_fr_r_04_capsule, center=<6 0 9>, rotation=<rot 0 0 90>}]
 dyn_coat_fr_r_05 [{shapeType=Capsule, length=16.75, radius=10.5, shapeName=dyn_coat_fr_r_05_capsule, center=<9 0 9>, rotation=<rot 0 0 90>}]
 dyn_coat_fr_r_MSTR [{shapeType=Capsule, length=0.505, radius=0.505}]
 dyn_coat_l_01 [{shapeType=Capsule, length=9.375, radius=10, center=<4.017 -3.6758 8.2178>, rotation=<rot 65.902 -89.9996 -89.9996>}]
 dyn_coat_l_02 [{shapeType=Capsule, length=1.875, radius=10, center=<3.8636 -3.6758 8.2177>, rotation=<rot 65.902 -89.9996 -89.9996>}]
 dyn_coat_l_03 [{shapeType=Capsule, length=16.75, radius=10.5, center=<6 0 9>, rotation=<rot 0 0 90>}]
 dyn_coat_l_04 [{shapeType=Capsule, length=16.75, radius=10.5, center=<6 0 9>, rotation=<rot 0 0 90>}]
 dyn_coat_l_05 [{shapeType=Capsule, length=16.75, radius=10.5, center=<6 0 9>, rotation=<rot 0 0 90>}]
 dyn_coat_l_MSTR [{shapeType=Capsule, length=0.505, radius=0.505}]
 dyn_coat_r_01 [{shapeType=Capsule, length=9.2812, radius=10.25, center=<4.2067 0 9.5255>, rotation=<rot 90 0 0>}]
 dyn_coat_r_02 [{shapeType=Capsule, length=1.5, radius=10.25, center=<2.7235 0.0001 9.5255>, rotation=<rot 90 0 0>}]
 dyn_coat_r_03 [{shapeType=Capsule, length=16.75, radius=10.5, center=<6 0 9>, rotation=<rot 0 0 90>}]
 dyn_coat_r_04 [{shapeType=Capsule, length=16.75, radius=10.5, center=<6 0 9>, rotation=<rot 0 0 90>}]
 dyn_coat_r_05 [{shapeType=Capsule, length=16.75, radius=10.5, center=<6 0 9>, rotation=<rot 0 0 90>}]
 dyn_coat_r_MSTR [{shapeType=Capsule, length=0.505, radius=0.505}]
 dyn_pants_string_01 [{shapeType=Capsule, length=3.2844, radius=0.5, center=<2.0266 0 0>, rotation=<rot 90 0 0>}]
 dyn_pants_string_02 [{shapeType=Capsule, length=5.0344, radius=0.875, center=<3.0915 0.0001 0>, rotation=<rot 90 0 0>}]
 dyn_pants_string_03 [{shapeType=Capsule, length=5.0344, radius=0.875, center=<3.0915 0.0001 0>, rotation=<rot 90 0 0>}]
 dyn_pants_string_MSTR [{shapeType=Capsule, length=0.505, radius=0.505}]
 dyn_zipper_fanny [{shapeType=Capsule, length=2.0031, radius=0.5, center=<1.6421 0 0>, rotation=<rot 90 0 0>}]
 dyn_zipper_fanny_MSTR [{shapeType=Capsule, length=0.505, radius=0.505}]
 pelvis [{shapeType=Capsule, length=8.9375, radius=13.5625, center=<-1.0665 0.9852 0>}]
 spine_01 [{shapeType=Capsule, length=6.8001, radius=14.5228, center=<-3.4316 2.39 -0.4237>}]
 spine_02 [{shapeType=Capsule, length=4.5346, radius=13.2728, center=<-2.4193 3.2691 0.8186>}]
 thigh_l [{shapeType=Capsule, length=25.0479, radius=13.4459, center=<-19.3575 0.3623 2.9475>, rotation=<rot 44.9573 -90.1174 -90.7771>}]
 thigh_r [
  {shapeType=Capsule, length=15.3916, radius=5.8834, center=<9.4849 -4.4974 5.338>, rotation=<rot 82.6955 -110.6022 -119.6249>}
  {shapeType=Capsule, length=25.6416, radius=12.8521, center=<19.3461 -0.2395 -1.8601>, rotation=<rot 44.9573 -90.1174 -90.7771>}
 ]
}
constraints [
 {
  bone1Name=dyn_coat_bk_l_MSTR
  bone2Name=spine_02
  swing1LimitDegrees=45
  swing1Motion=Free
  swing2LimitDegrees=45
  swing2Motion=Free
  twistLimitDegrees=45
  twistMotion=Free
 }
 {bone1Name=dyn_coat_bk_l_01, bone2Name=dyn_coat_bk_l_MSTR, swing1LimitDegrees=2, swing2LimitDegrees=15, twistLimitDegrees=2}
 {bone1Name=dyn_coat_bk_l_02, bone2Name=dyn_coat_bk_l_01, swing1LimitDegrees=2, swing2LimitDegrees=15, twistLimitDegrees=2}
 {bone1Name=dyn_coat_bk_l_03, bone2Name=dyn_coat_bk_l_02, swing1LimitDegrees=2, swing2LimitDegrees=15, twistLimitDegrees=2}
 {bone1Name=dyn_coat_bk_l_04, bone2Name=dyn_coat_bk_l_03, swing1LimitDegrees=2, swing2LimitDegrees=15, twistLimitDegrees=2}
 {bone1Name=dyn_coat_bk_l_05, bone2Name=dyn_coat_bk_l_04, swing1LimitDegrees=2, swing2LimitDegrees=15, twistLimitDegrees=2}
 {
  bone1Name=dyn_coat_l_MSTR
  bone2Name=spine_02
  swing1LimitDegrees=45
  swing1Motion=Free
  swing2LimitDegrees=45
  swing2Motion=Free
  twistLimitDegrees=45
  twistMotion=Free
 }
 {bone1Name=dyn_coat_l_01, bone2Name=dyn_coat_l_MSTR, swing1LimitDegrees=2, swing2LimitDegrees=15, twistLimitDegrees=2}
 {bone1Name=dyn_coat_l_02, bone2Name=dyn_coat_l_01, swing1LimitDegrees=2, swing2LimitDegrees=15, twistLimitDegrees=2}
 {bone1Name=dyn_coat_l_03, bone2Name=dyn_coat_l_02, swing1LimitDegrees=2, swing2LimitDegrees=15, twistLimitDegrees=2}
 {bone1Name=dyn_coat_l_04, bone2Name=dyn_coat_l_03, swing1LimitDegrees=2, swing2LimitDegrees=15, twistLimitDegrees=2}
 {bone1Name=dyn_coat_l_05, bone2Name=dyn_coat_l_04, swing1LimitDegrees=2, swing2LimitDegrees=15, twistLimitDegrees=2}
 {
  bone1Name=dyn_coat_fr_l_MSTR
  bone2Name=spine_02
  swing1LimitDegrees=45
  swing1Motion=Free
  swing2LimitDegrees=45
  swing2Motion=Free
  twistLimitDegrees=45
  twistMotion=Free
 }
 {bone1Name=dyn_coat_fr_l_01, bone2Name=dyn_coat_fr_l_MSTR, swing1LimitDegrees=2, swing2LimitDegrees=15, twistLimitDegrees=2}
 {bone1Name=dyn_coat_fr_l_02, bone2Name=dyn_coat_fr_l_01, swing1LimitDegrees=2, swing2LimitDegrees=15, twistLimitDegrees=2}
 {bone1Name=dyn_coat_fr_l_03, bone2Name=dyn_coat_fr_l_02, swing1LimitDegrees=2, swing2LimitDegrees=15, twistLimitDegrees=2}
 {bone1Name=dyn_coat_fr_l_04, bone2Name=dyn_coat_fr_l_03, swing1LimitDegrees=2, swing2LimitDegrees=15, twistLimitDegrees=2}
 {bone1Name=dyn_coat_fr_l_05, bone2Name=dyn_coat_fr_l_04, swing1LimitDegrees=2, swing2LimitDegrees=15, twistLimitDegrees=2}
 {
  bone1Name=dyn_coat_fr_r_MSTR
  bone2Name=spine_02
  swing1LimitDegrees=45
  swing1Motion=Free
  swing2LimitDegrees=45
  swing2Motion=Free
  twistLimitDegrees=45
  twistMotion=Free
 }
 {bone1Name=dyn_coat_fr_r_01, bone2Name=dyn_coat_fr_r_MSTR, swing1LimitDegrees=2, swing2LimitDegrees=15, twistLimitDegrees=2}
 {bone1Name=dyn_coat_fr_r_02, bone2Name=dyn_coat_fr_r_01, swing1LimitDegrees=2, swing2LimitDegrees=15, twistLimitDegrees=2}
 {bone1Name=dyn_coat_fr_r_03, bone2Name=dyn_coat_fr_r_02, swing1LimitDegrees=2, swing2LimitDegrees=15, twistLimitDegrees=2}
 {bone1Name=dyn_coat_fr_r_04, bone2Name=dyn_coat_fr_r_03, swing1LimitDegrees=2, swing2LimitDegrees=15, twistLimitDegrees=2}
 {bone1Name=dyn_coat_fr_r_05, bone2Name=dyn_coat_fr_r_04, swing1LimitDegrees=2, swing2LimitDegrees=15, twistLimitDegrees=2}
 {
  bone1Name=dyn_coat_r_MSTR
  bone2Name=spine_02
  swing1LimitDegrees=45
  swing1Motion=Free
  swing2LimitDegrees=45
  swing2Motion=Free
  twistLimitDegrees=45
  twistMotion=Free
 }
 {bone1Name=dyn_coat_r_01, bone2Name=dyn_coat_r_MSTR, swing1LimitDegrees=2, swing2LimitDegrees=15, twistLimitDegrees=2}
 {bone1Name=dyn_coat_r_02, bone2Name=dyn_coat_r_01, swing1LimitDegrees=2, swing2LimitDegrees=15, twistLimitDegrees=2}
 {bone1Name=dyn_coat_r_03, bone2Name=dyn_coat_r_02, swing1LimitDegrees=2, swing2LimitDegrees=15, twistLimitDegrees=2}
 {bone1Name=dyn_coat_r_04, bone2Name=dyn_coat_r_03, swing1LimitDegrees=2, swing2LimitDegrees=15, twistLimitDegrees=2}
 {bone1Name=dyn_coat_r_05, bone2Name=dyn_coat_r_04, swing1LimitDegrees=2, swing2LimitDegrees=15, twistLimitDegrees=2}
 {
  bone1Name=dyn_coat_bk_MSTR
  bone2Name=spine_02
  swing1LimitDegrees=45
  swing1Motion=Free
  swing2LimitDegrees=45
  swing2Motion=Free
  twistLimitDegrees=45
  twistMotion=Free
 }
 {bone1Name=dyn_coat_bk_01, bone2Name=dyn_coat_bk_MSTR, swing1LimitDegrees=2, swing2LimitDegrees=15, twistLimitDegrees=2}
 {bone1Name=dyn_coat_bk_02, bone2Name=dyn_coat_bk_01, swing1LimitDegrees=2, swing2LimitDegrees=15, twistLimitDegrees=2}
 {bone1Name=dyn_coat_bk_03, bone2Name=dyn_coat_bk_02, swing1LimitDegrees=2, swing2LimitDegrees=15, twistLimitDegrees=2}
 {
  bone1Name=dyn_coat_bk_r_MSTR
  bone2Name=spine_02
  swing1LimitDegrees=45
  swing1Motion=Free
  swing2LimitDegrees=45
  swing2Motion=Free
  twistLimitDegrees=45
  twistMotion=Free
 }
 {bone1Name=dyn_coat_bk_r_01, bone2Name=dyn_coat_bk_r_MSTR, swing1LimitDegrees=2, swing2LimitDegrees=15, twistLimitDegrees=2}
 {bone1Name=dyn_coat_bk_r_02, bone2Name=dyn_coat_bk_r_01, swing1LimitDegrees=2, swing2LimitDegrees=15, twistLimitDegrees=2}
 {bone1Name=dyn_coat_bk_r_03, bone2Name=dyn_coat_bk_r_02, swing1LimitDegrees=2, swing2LimitDegrees=15, twistLimitDegrees=2}
 {bone1Name=dyn_coat_bk_r_04, bone2Name=dyn_coat_bk_r_03, swing1LimitDegrees=2, swing2LimitDegrees=15, twistLimitDegrees=2}
 {bone1Name=dyn_coat_bk_r_05, bone2Name=dyn_coat_bk_r_04, swing1LimitDegrees=2, swing2LimitDegrees=15, twistLimitDegrees=2}
 {
  bone1Name=spine_01
  bone2Name=pelvis
  swing1LimitDegrees=45
  swing1Motion=Free
  swing2LimitDegrees=45
  swing2Motion=Free
  twistLimitDegrees=45
  twistMotion=Free
 }
 {
  bone1Name=spine_02
  bone2Name=spine_01
  swing1LimitDegrees=45
  swing1Motion=Free
  swing2LimitDegrees=45
  swing2Motion=Free
  twistLimitDegrees=45
  twistMotion=Free
 }
 {
  bone1Name=thigh_l
  bone2Name=pelvis
  swing1LimitDegrees=45
  swing1Motion=Free
  swing2LimitDegrees=45
  swing2Motion=Free
  twistLimitDegrees=45
  twistMotion=Free
 }
 {
  bone1Name=calf_l
  bone2Name=thigh_l
  swing1LimitDegrees=45
  swing1Motion=Free
  swing2LimitDegrees=45
  swing2Motion=Free
  twistLimitDegrees=45
  twistMotion=Free
 }
 {
  bone1Name=thigh_r
  bone2Name=pelvis
  swing1LimitDegrees=45
  swing1Motion=Free
  swing2LimitDegrees=45
  swing2Motion=Free
  twistLimitDegrees=45
  twistMotion=Free
 }
 {
  bone1Name=calf_r
  bone2Name=thigh_r
  swing1LimitDegrees=45
  swing1Motion=Free
  swing2LimitDegrees=45
  swing2Motion=Free
  twistLimitDegrees=45
  twistMotion=Free
 }
 {
  bone1Name=dyn_coat_l_03
  bone2Name=dyn_coat_fr_l_03
  swing1LimitDegrees=45
  swing1Motion=Free
  swing2LimitDegrees=45
  swing2Motion=Free
  twistLimitDegrees=45
  twistMotion=Free
 }
 {
  bone1Name=dyn_coat_l_05
  bone2Name=dyn_coat_fr_l_05
  swing1LimitDegrees=45
  swing1Motion=Free
  swing2LimitDegrees=45
  swing2Motion=Free
  twistLimitDegrees=45
  twistMotion=Free
 }
 {
  bone1Name=dyn_coat_bk_l_03
  bone2Name=dyn_coat_l_03
  swing1LimitDegrees=45
  swing1Motion=Free
  swing2LimitDegrees=45
  swing2Motion=Free
  twistLimitDegrees=45
  twistMotion=Free
 }
 {
  bone1Name=dyn_coat_r_03
  bone2Name=dyn_coat_fr_r_03
  swing1LimitDegrees=45
  swing1Motion=Free
  swing2LimitDegrees=45
  swing2Motion=Free
  twistLimitDegrees=45
  twistMotion=Free
 }
 {
  bone1Name=dyn_coat_r_05
  bone2Name=dyn_coat_fr_r_05
  swing1LimitDegrees=45
  swing1Motion=Free
  swing2LimitDegrees=45
  swing2Motion=Free
  twistLimitDegrees=45
  twistMotion=Free
 }
 {
  bone1Name=dyn_coat_bk_r_03
  bone2Name=dyn_coat_r_03
  swing1LimitDegrees=45
  swing1Motion=Free
  swing2LimitDegrees=45
  swing2Motion=Free
  twistLimitDegrees=45
  twistMotion=Free
 }
 {bone1Name=dyn_coat_bk_04, bone2Name=dyn_coat_bk_03, swing1LimitDegrees=2, swing2LimitDegrees=15, twistLimitDegrees=2}
 {bone1Name=dyn_coat_bk_05, bone2Name=dyn_coat_bk_04, swing1LimitDegrees=2, swing2LimitDegrees=15, twistLimitDegrees=2}
 {
  bone1Name=dyn_zipper_fanny_MSTR
  bone2Name=pelvis
  swing1LimitDegrees=45
  swing1Motion=Free
  swing2LimitDegrees=45
  swing2Motion=Free
  twistLimitDegrees=45
  twistMotion=Free
 }
 {bone1Name=dyn_zipper_fanny, bone2Name=dyn_zipper_fanny_MSTR, swing1LimitDegrees=40, swing2LimitDegrees=20, twistLimitDegrees=5}
 {
  bone1Name=dyn_pants_string_MSTR
  bone2Name=pelvis
  swing1LimitDegrees=45
  swing1Motion=Free
  swing2LimitDegrees=45
  swing2Motion=Free
  twistLimitDegrees=45
  twistMotion=Free
 }
 {bone1Name=dyn_pants_string_01, bone2Name=dyn_pants_string_MSTR, swing1LimitDegrees=40, swing2LimitDegrees=40, twistLimitDegrees=5}
 {bone1Name=dyn_pants_string_02, bone2Name=dyn_pants_string_01, swing1LimitDegrees=40, swing2LimitDegrees=40, twistLimitDegrees=5}
 {bone1Name=dyn_pants_string_03, bone2Name=dyn_pants_string_01, swing1LimitDegrees=40, swing2LimitDegrees=40, twistLimitDegrees=5}
 {
  bone1Name=dyn_coat_bk_r_05
  bone2Name=dyn_coat_bk_05
  swing1LimitDegrees=45
  swing1Motion=Free
  swing2LimitDegrees=45
  swing2Motion=Free
  twistLimitDegrees=45
  twistMotion=Free
 }
 {
  bone1Name=dyn_coat_bk_l_05
  bone2Name=dyn_coat_bk_05
  swing1LimitDegrees=45
  swing1Motion=Free
  swing2LimitDegrees=45
  swing2Motion=Free
  twistLimitDegrees=45
  twistMotion=Free
 }
]
