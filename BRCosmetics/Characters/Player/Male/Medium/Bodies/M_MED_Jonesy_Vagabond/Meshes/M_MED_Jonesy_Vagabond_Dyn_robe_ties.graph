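# Physics asset for the secondary motion of the back coat panels and rope ties on the male Jonesy Vagabond body.
# Left and right coat chains each run six capsule bodies from a small master, with wide swing and tight twist limits.
# Five-segment rope chains hang from masters angularly locked to each side's second coat bone, with one swing axis much looser than the other.
# Coat capsules are large, offset and angled, far bigger than the rope capsules.
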PhysicsAsset /BRCosmetics/Characters/Player/Male/Medium/Bodies/M_MED_Jonesy_Vagabond/Meshes/M_MED_Jonesy_Vagabond_Dyn_robe_ties
bodies {
 dyn_coat_bk_l_01 [{shapeType=Capsule, length=40.625, radius=38, center=<10.0703 12.3057 -34.8129>, rotation=<rot 0 0 70.0001>}]
 dyn_coat_bk_l_02 [{shapeType=Capsule, length=40.625, radius=38, center=<10.0703 12.3057 -34.8129>, rotation=<rot 0 0 70.0001>}]
 dyn_coat_bk_l_03 [{shapeType=Capsule, length=40.625, radius=38, center=<10.0703 12.3057 -34.8129>, rotation=<rot 0 0 70.0001>}]
 dyn_coat_bk_l_04 [{shapeType=Capsule, length=40.625, radius=38, center=<10.0703 11.8866 -33.6614>, rotation=<rot 0 0 70.0001>}]
 dyn_coat_bk_l_05 [{shapeType=Capsule, length=40.625, radius=38, center=<10.0703 11.8306 -33.5069>, rotation=<rot 0 0 70.0001>}]
 dyn_coat_bk_l_06 [{shapeType=Capsule, length=40.625, radius=38, center=<10.0703 11.7188 -33.2006>, rotation=<rot 0 0 70.0001>}]
 dyn_coat_bk_l_MSTR [{shapeType=Capsule, length=0.505, radius=0.505}]
 dyn_coat_bk_r_01 [{shapeType=Capsule, length=40.625, radius=38, center=<10.0703 -12.3057 -34.8129>, rotation=<rot 0 0 -70.0001>}]
 dyn_coat_bk_r_02 [{shapeType=Capsule, length=40.625, radius=38, center=<10.0703 -12.3057 -34.8129>, rotation=<rot 0 0 -70.0001>}]
 dyn_coat_bk_r_03 [{shapeType=Capsule, length=40.625, radius=38, center=<10.0703 -11.9824 -33.9244>, rotation=<rot 0 0 -70.0001>}]
 dyn_coat_bk_r_04 [{shapeType=Capsule, length=40.625, radius=38, center=<10.0703 -11.7068 -33.1682>, rotation=<rot 0 0 -70.0001>}]
 dyn_coat_bk_r_05 [{shapeType=Capsule, length=40.625, radius=38, center=<10.0703 -11.7311 -33.2346>, rotation=<rot 0 0 -70.0001>}]
 dyn_coat_bk_r_06 [{shapeType=Capsule, length=40.625, radius=38, center=<10.0703 -11.715 -33.1896>, rotation=<rot 0 0 -70.0001>}]
 dyn_coat_bk_r_MSTR [{shapeType=Capsule, length=0.505, radius=0.505}]
 dyn_robe_rope_l_01 [{shapeType=Capsule, length=3.9375, radius=3.3125, center=<6.6282 0 2.7>, rotation=<rot 85 -90.0016 -90.0002>}]
 dyn_robe_rope_l_02 [{shapeType=Capsule, length=7.8438, radius=4.4062, center=<4.3798 0 3.25>, rotation=<rot 84.9997 -90.0013 -89.9998>}]
 dyn_robe_rope_l_03 [{shapeType=Capsule, length=7.8438, radius=4.4062, center=<4.3849 0 3.25>, rotation=<rot 84.9998 -90.0016 -90.0002>}]
 dyn_robe_rope_l_04 [{shapeType=Capsule, length=7.8438, radius=4.4062, center=<4.3894 0 3.25>, rotation=<rot 85 -90.0019 -90.0003>}]
 dyn_robe_rope_l_05 [{shapeType=Capsule, length=3.2188, radius=4.4062, center=<1.8881 0 3.25>, rotation=<rot 84.9997 -90.0013 -89.9998>}]
 dyn_robe_rope_l_MSTR [{shapeType=Capsule, length=0.505, radius=0.505}]
 dyn_robe_rope_r_01 [{shapeType=Capsule, length=3.9375, radius=3.3125, center=<6.7357 0.3606 2.7>, rotation=<rot 79.9989 89.9995 89.9995>}]
 dyn_robe_rope_r_02 [{shapeType=Capsule, length=7.8438, radius=4.4062, center=<4.384 0.4983 3.25>, rotation=<rot 79.9989 89.9996 89.9996>}]
 dyn_robe_rope_r_03 [{shapeType=Capsule, length=7.8438, radius=4.4062, center=<4.3818 0.4334 3.25>, rotation=<rot 79.9989 89.9995 89.9995>}]
 dyn_robe_rope_r_04 [{shapeType=Capsule, length=7.8438, radius=4.4062, center=<4.398 0.4334 3.25>, rotation=<rot 79.9989 89.9995 89.9995>}]
 dyn_robe_rope_r_05 [{shapeType=Capsule, length=3.2188, radius=4.4062, center=<1.8882 0.4335 3.25>, rotation=<rot 79.9989 89.9995 89.9995>}]
 dyn_robe_rope_r_MSTR [{shapeType=Capsule, length=0.505, radius=0.505}]
}
constraints [
 {bone1Name=dyn_coat_bk_l_01, bone2Name=dyn_coat_bk_l_MSTR, swing1LimitDegrees=45, swing2LimitDegrees=45, twistLimitDegrees=5}
 {bone1Name=dyn_coat_bk_l_02, bone2Name=dyn_coat_bk_l_01, swing1LimitDegrees=45, swing2LimitDegrees=45, twistLimitDegrees=5}
 {bone1Name=dyn_coat_bk_l_03, bone2Name=dyn_coat_bk_l_02, swing1LimitDegrees=45, swing2LimitDegrees=45, twistLimitDegrees=5}
 {bone1Name=dyn_coat_bk_l_04, bone2Name=dyn_coat_bk_l_03, swing1LimitDegrees=45, swing2LimitDegrees=45, twistLimitDegrees=5}
 {bone1Name=dyn_coat_bk_l_05, bone2Name=dyn_coat_bk_l_04, swing1LimitDegrees=45, swing2LimitDegrees=45, twistLimitDegrees=5}
 {bone1Name=dyn_coat_bk_l_06, bone2Name=dyn_coat_bk_l_05, swing1LimitDegrees=45, swing2LimitDegrees=45, twistLimitDegrees=5}
 {bone1Name=dyn_coat_bk_r_01, bone2Name=dyn_coat_bk_r_MSTR, swing1LimitDegrees=45, swing2LimitDegrees=45, twistLimitDegrees=5}
 {bone1Name=dyn_coat_bk_r_02, bone2Name=dyn_coat_bk_r_01, swing1LimitDegrees=45, swing2LimitDegrees=45, twistLimitDegrees=5}
 {bone1Name=dyn_coat_bk_r_03, bone2Name=dyn_coat_bk_r_02, swing1LimitDegrees=45, swing2LimitDegrees=45, twistLimitDegrees=5}
 {bone1Name=dyn_coat_bk_r_04, bone2Name=dyn_coat_bk_r_03, swing1LimitDegrees=45, swing2LimitDegrees=45, twistLimitDegrees=5}
 {bone1Name=dyn_coat_bk_r_05, bone2Name=dyn_coat_bk_r_04, swing1LimitDegrees=45, swing2LimitDegrees=45, twistLimitDegrees=5}
 {bone1Name=dyn_coat_bk_r_06, bone2Name=dyn_coat_bk_r_05, swing1LimitDegrees=45, swing2LimitDegrees=45, twistLimitDegrees=5}
 {
  bone1Name=dyn_robe_rope_r_MSTR
  bone2Name=dyn_coat_bk_r_02
  swing1LimitDegrees=45
  swing1Motion=Locked
  swing2LimitDegrees=45
  swing2Motion=Locked
  twistLimitDegrees=45
  twistMotion=Locked
 }
 {bone1Name=dyn_robe_rope_r_01, bone2Name=dyn_robe_rope_r_MSTR, swing1LimitDegrees=5, swing2LimitDegrees=45, twistLimitDegrees=5}
 {bone1Name=dyn_robe_rope_r_02, bone2Name=dyn_robe_rope_r_01, swing1LimitDegrees=5, swing2LimitDegrees=45, twistLimitDegrees=5}
 {bone1Name=dyn_robe_rope_r_03, bone2Name=dyn_robe_rope_r_02, swing1LimitDegrees=5, swing2LimitDegrees=45, twistLimitDegrees=5}
 {bone1Name=dyn_robe_rope_r_04, bone2Name=dyn_robe_rope_r_03, swing1LimitDegrees=5, swing2LimitDegrees=45, twistLimitDegrees=5}
 {bone1Name=dyn_robe_rope_r_05, bone2Name=dyn_robe_rope_r_04, swing1LimitDegrees=5, swing2LimitDegrees=45, twistLimitDegrees=5}
 {
  bone1Name=dyn_robe_rope_l_MSTR
  bone2Name=dyn_coat_bk_l_02
  swing1LimitDegrees=45
  swing1Motion=Locked
  swing2LimitDegrees=45
  swing2Motion=Locked
  twistLimitDegrees=45
  twistMotion=Locked
 }
 {bone1Name=dyn_robe_rope_l_01, bone2Name=dyn_robe_rope_l_MSTR, swing1LimitDegrees=5, swing2LimitDegrees=45, twistLimitDegrees=5}
 {bone1Name=dyn_robe_rope_l_02, bone2Name=dyn_robe_rope_l_01, swing1LimitDegrees=5, swing2LimitDegrees=45, twistLimitDegrees=5}
 {bone1Name=dyn_robe_rope_l_03, bone2Name=dyn_robe_rope_l_02, swing1LimitDegrees=5, swing2LimitDegrees=45, twistLimitDegrees=5}
 {bone1Name=dyn_robe_rope_l_04, bone2Name=dyn_robe_rope_l_03, swing1LimitDegrees=5, swing2LimitDegrees=45, twistLimitDegrees=5}
 {bone1Name=dyn_robe_rope_l_05, bone2Name=dyn_robe_rope_l_04, swing1LimitDegrees=5, swing2LimitDegrees=45, twistLimitDegrees=5}
]
