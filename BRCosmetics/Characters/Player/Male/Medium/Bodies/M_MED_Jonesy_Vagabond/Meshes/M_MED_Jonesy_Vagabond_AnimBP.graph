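# Anim Blueprint for the Jonesy Vagabond body that copies the parent mesh pose and adds secondary coat motion.
# Slope-adjusted foot bones feed RigidBody robe sims, switched for skydiving, with state-driven gravity and thigh-driven front-left coat offset.
# Driving certain vehicles shrinks coat bones and disables the standard robe sim; exiting or targeting resets dynamics.
# Accessory constraint chains and other RigidBody nodes are present but disconnected from the output.
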AnimBlueprint /BRCosmetics/Characters/Player/Male/Medium/Bodies/M_MED_Jonesy_Vagabond/Meshes/M_MED_Jonesy_Vagabond_AnimBP
target_skeleton=&/M_MED_Jonesy_Vagabond_Skeleton
graphs [
 AnimGraph {nodes [
   n1 = "Constraint - dyn_gun_r_MSTR (PPP)"!(ComponentPose, Alpha=1, ConstraintWeights_0=0.0417, ConstraintWeights_1:float=0.1728, ConstraintWeights_2:float=0.7855) -> (Pose) {
    alphaBoolBlend {blendOption=Linear}
    boneToModify {boneName=dyn_gun_r_MSTR}
    constraintSetup [{targetBone {boneName=thigh_r}}, {targetBone {boneName=pelvis}}, {targetBone {boneName=thigh_twist_01_r}}]
    constraintWeights [1,1,1]
   }
   n2 = "Constraint - dyn_pants_string_MSTR (PPPP)"!(
    ComponentPose=@n1
    Alpha=1
    ConstraintWeights_0=0.8049
    ConstraintWeights_1:float=0.0126
    ConstraintWeights_2:float=0.0333
    ConstraintWeights_3:float=0.1492
   ) -> (Pose) {
    alphaBoolBlend {blendOption=Linear}
    boneToModify {boneName=dyn_pants_string_MSTR}
    constraintSetup [
     {targetBone {boneName=pelvis}}
     {targetBone {boneName=thigh_twist_01_l}}
     {targetBone {boneName=thigh_twist_01_r}}
     {targetBone {boneName=spine_01}}
    ]
    constraintWeights [1,1,1,1]
   }
   n3 = "Constraint - dyn_zipper_fanny_MSTR (PPP)"!(ComponentPose=@n2, Alpha=1, ConstraintWeights_0=0.6127, ConstraintWeights_1:float=0.2715, ConstraintWeights_2:float=0.1158) -> (Pose) {
    alphaBoolBlend {blendOption=Linear}
    boneToModify {boneName=dyn_zipper_fanny_MSTR}
    constraintSetup [{targetBone {boneName=pelvis}}, {targetBone {boneName=thigh_twist_01_l}}, {targetBone {boneName=spine_01}}]
    constraintWeights [1,1,1]
   }
   n4 = "Constraint - dyn_pouch_chest_l_MSTR (PPPP)"!(
    ComponentPose=@n3
    Alpha=1
    ConstraintWeights_0=0.0807
    ConstraintWeights_1:float=0.2248
    ConstraintWeights_2:float=0.5013
    ConstraintWeights_3:float=0.1931
   ) -> (Pose) {
    alphaBoolBlend {blendOption=Linear}
    boneToModify {boneName=dyn_pouch_chest_l_MSTR}
    constraintSetup [{targetBone {boneName=spine_02}}, {targetBone {boneName=spine_03}}, {targetBone {boneName=spine_04}}, {targetBone {boneName=spine_05}}]
    constraintWeights [1,1,1,1]
   }
   n5 = "Constraint - dyn_pouch_chest_r_MSTR (PPPP)"!(
    ComponentPose=@n4
    Alpha=1
    ConstraintWeights_0=0.1332
    ConstraintWeights_1:float=0.2665
    ConstraintWeights_2:float=0.4708
    ConstraintWeights_3:float=0.1295
   ) -> (Pose) {
    alphaBoolBlend {blendOption=Linear}
    boneToModify {boneName=dyn_pouch_chest_r_MSTR}
    constraintSetup [{targetBone {boneName=spine_02}}, {targetBone {boneName=spine_03}}, {targetBone {boneName=spine_04}}, {targetBone {boneName=spine_05}}]
    constraintWeights [1,1,1,1]
   }
   n6 = "Constraint - dyn_zipper_chest_MSTR (PPPP)"!(
    ComponentPose=@n5
    Alpha=1
    ConstraintWeights_0=0.0155
    ConstraintWeights_1:float=0.0835
    ConstraintWeights_2:float=0.2524
    ConstraintWeights_3:float=0.6487
   ) -> (Pose) {
    alphaBoolBlend {blendOption=Linear}
    boneToModify {boneName=dyn_zipper_chest_MSTR}
    constraintSetup [{targetBone {boneName=neck_02}}, {targetBone {boneName=spine_03}}, {targetBone {boneName=spine_04}}, {targetBone {boneName=spine_05}}]
    constraintWeights [1,1,1,1]
   }
   "Constraint - dyn_zipper_back_MSTR (PPPP)"!(
    ComponentPose=@n6
    Alpha=1
    ConstraintWeights_0=0.0225
    ConstraintWeights_1:float=0.1622
    ConstraintWeights_2:float=0.2444
    ConstraintWeights_3:float=0.5708
   ) -> (Pose) {
    alphaBoolBlend {blendOption=Linear}
    boneToModify {boneName=dyn_zipper_back_MSTR}
    constraintSetup [{targetBone {boneName=clavicle_r}}, {targetBone {boneName=spine_03}}, {targetBone {boneName=spine_04}}, {targetBone {boneName=spine_05}}]
    constraintWeights [1,1,1,1]
   }
   n7 = RigidBody() {
    bClampLinearTranslationLimitToRefPose=true
    bOverrideWorldGravity=true
    overridePhysicsAsset=&/M_MED_Jonesy_Vagabond_Dyn_leg_l
    simulationSpace=BaseBoneSpace
    alphaBoolBlend {blendOption=Linear}
    baseBoneRef {boneName=calf_l}
    componentAppliedLinearAccClamp=<3000 3000 3000>
    componentLinearAccScale=<0.4 0.3 0.3>
    componentLinearVelScale=<0.4 0.3 0.3>
    overrideWorldGravity {z=-980}
   }
   RigidBody(ComponentPose=@n7) {
    bClampLinearTranslationLimitToRefPose=true
    bOverrideWorldGravity=true
    overridePhysicsAsset=&/M_MED_Jonesy_Vagabond_Dyn_leg_r
    simulationSpace=BaseBoneSpace
    alphaBoolBlend {blendOption=Linear}
    baseBoneRef {boneName=calf_r}
    componentAppliedLinearAccClamp=<3000 3000 3000>
    componentLinearAccScale=<0.4 0.3 0.3>
    componentLinearVelScale=<0.4 0.3 0.3>
    overrideWorldGravity {z=-980}
   }
   RigidBody() {
    bClampLinearTranslationLimitToRefPose=true
    bOverrideWorldGravity=true
    overridePhysicsAsset=&/M_MED_Jonesy_Vagabond_Dyn_fries
    simulationSpace=BaseBoneSpace
    alphaBoolBlend {blendOption=Linear}
    baseBoneRef {boneName=spine_02}
    componentAppliedLinearAccClamp=<3000 3000 3000>
    componentLinearAccScale=<0.5 0.3 0.3>
    componentLinearVelScale=<0.5 0.3 0.3>
    overrideWorldGravity {z=-980}
   }
   n8 = Get gravity_override() -> (gravity_override:Vector)
   RigidBody!(ComponentPose, Alpha, OverrideWorldGravity:Vector=@n8, ExternalForce) -> (Pose) {
    bClampLinearTranslationLimitToRefPose=true
    bOverrideWorldGravity=true
    overridePhysicsAsset=&/M_MED_Jonesy_Vagabond_Dyn
    simulationSpace=BaseBoneSpace
    alphaBoolBlend {blendOption=Linear}
    baseBoneRef {boneName=spine_04}
    componentAppliedLinearAccClamp=<3000 3000 3000>
    componentLinearAccScale=<0.4 0.3 0.3>
    componentLinearVelScale=<0.4 0.3 0.3>
    overrideWorldGravity {z=-980}
   }
   n9 = "Constraint - dyn_fries_cup_MSTR (PPPPPPPP)"!(
    ComponentPose
    Alpha=1
    ConstraintWeights_0=0.1108
    ConstraintWeights_1:float=0.0247
    ConstraintWeights_2:float=0.3122
    ConstraintWeights_3:float=0.3762
    ConstraintWeights_4:float=0.0429
    ConstraintWeights_5:float=0.0955
    ConstraintWeights_6:float=0.0264
    ConstraintWeights_7:float=0.0113
   ) -> (Pose) {
    alphaBoolBlend {blendOption=Linear}
    boneToModify {boneName=dyn_fries_cup_MSTR}
    constraintSetup [
     {targetBone {boneName=dyn_coat_l_02}}
     {targetBone {boneName=dyn_coat_l_03}}
     {targetBone {boneName=dyn_coat_fr_l_02}}
     {targetBone {boneName=dyn_coat_fr_l_03}}
     {targetBone {boneName=dyn_coat_fr_l_04}}
     {targetBone {boneName=dyn_coat_bk_l_02}}
     {targetBone {boneName=dyn_coat_bk_l_03}}
     {targetBone {boneName=spine_03}}
    ]
    constraintWeights [1,1,1,1,1,1,1,1]
   }
   n10 = "Constraint - dyn_robe_rope_l_MSTR (PPPPPPP)"!(
    ComponentPose=@n9
    Alpha=1
    ConstraintWeights_0=0.1317
    ConstraintWeights_1:float=0.014
    ConstraintWeights_2:float=0.1256
    ConstraintWeights_3:float=0.2372
    ConstraintWeights_4:float=0.0335
    ConstraintWeights_5:float=0.0677
    ConstraintWeights_6:float=0.3903
   ) -> (Pose) {
    alphaBoolBlend {blendOption=Linear}
    boneToModify {boneName=dyn_robe_rope_l_MSTR}
    constraintSetup [
     {targetBone {boneName=dyn_coat_bk_02}}
     {targetBone {boneName=dyn_coat_bk_03}}
     {targetBone {boneName=dyn_coat_bk_01}}
     {targetBone {boneName=dyn_coat_l_02}}
     {targetBone {boneName=dyn_coat_l_01}}
     {targetBone {boneName=dyn_coat_bk_l_03}}
     {targetBone {boneName=dyn_coat_bk_l_02}}
    ]
    constraintWeights [1,1,1,1,1,1,1]
   }
   n11 = "Constraint - dyn_robe_rope_r_MSTR (PPPP)"!(
    ComponentPose=@n10
    Alpha=1
    ConstraintWeights_0=0.144
    ConstraintWeights_1:float=0.0911
    ConstraintWeights_2:float=0.1907
    ConstraintWeights_3:float=0.5741
   ) -> (Pose) {
    alphaBoolBlend {blendOption=Linear}
    boneToModify {boneName=dyn_robe_rope_r_MSTR}
    constraintSetup [
     {targetBone {boneName=dyn_coat_r_02}}
     {targetBone {boneName=dyn_coat_bk_r_01}}
     {targetBone {boneName=dyn_coat_bk_r_03}}
     {targetBone {boneName=dyn_coat_bk_r_02}}
    ]
    constraintWeights [1,1,1,1]
   }
   RigidBody(ComponentPose=@n11, OverrideWorldGravity="0,0,0") {
    bClampLinearTranslationLimitToRefPose=true
    bOverrideWorldGravity=true
    overridePhysicsAsset=&/M_MED_Jonesy_Vagabond_Dyn_robe_ties
    simulationSpace=BaseBoneSpace
    alphaBoolBlend {blendOption=Linear}
    baseBoneRef {boneName=pelvis}
    componentAppliedLinearAccClamp=<3000 3000 3000>
    componentLinearAccScale=<0.4 0.3 0.3>
    componentLinearVelScale=<0.4 0.3 0.3>
    overrideWorldGravity {z=-980}
   }
   Output Pose(Result=Use cached pose 'skydivingResult'()) {layerGroup=None}
   n12 = NOT Boolean(A=Get bCanPlayCustomAnimations() -> (bCanPlayCustomAnimations:bool))
   n13 = Get slope_foot_rot() -> (slope_foot_rot:Rotator)
   n14 = "Transform (Modify) Bone - Bone: dyn_slope_foot_l"!(
    ComponentPose=Local To Component(LocalPose=Copy Pose From Mesh() {bUseAttachedParent=true})
    bAlphaBoolEnabled:bool=@n12
    Translation=Get slope_foot_l_trans() -> (slope_foot_l_trans:Vector)
    Rotation=@n13
    Scale="1,1,1"
   ) -> (Pose) {
    alphaInputType=Bool
    rotationMode=BMM_Replace
    rotationSpace=BCS_WorldSpace
    translationMode=BMM_Replace
    translationSpace=BCS_WorldSpace
    alphaBoolBlend {blendOption=Linear}
    boneToModify {boneName=dyn_slope_foot_l}
   }
   n15 = "Transform (Modify) Bone - Bone: dyn_slope_foot_r"!(
    ComponentPose=@n14
    bAlphaBoolEnabled:bool=@n12
    Translation=Get slope_foot_r_trans() -> (slope_foot_r_trans:Vector)
    Rotation=@n13
    Scale="1,1,1"
   ) -> (Pose) {
    alphaInputType=Bool
    rotationMode=BMM_Replace
    rotationSpace=BCS_WorldSpace
    translationMode=BMM_Replace
    translationSpace=BCS_WorldSpace
    alphaBoolBlend {blendOption=Linear}
    boneToModify {boneName=dyn_slope_foot_r}
   }
   n16 = Get in vehicle scale alpha() -> (in vehicle scale alpha:double)
   n17 = "Transform (Modify) Bone - Bone: dyn_coat_l_02"(ComponentPose=@n15, Alpha=@n16, Translation="0,0,0", Rotation="0,0,0", Scale="0.1,0.1,0.1") {scaleMode=BMM_Replace, alphaBoolBlend {blendOption=Linear}, boneToModify {boneName=dyn_coat_l_02}, scale=<0.1 0.1 0.1>}
   n18 = "Transform (Modify) Bone - Bone: dyn_coat_fr_l_02"(ComponentPose=@n17, Alpha=@n16, Translation="0,0,0", Rotation="0,0,0", Scale="0.1,0.1,0.1") {scaleMode=BMM_Replace, alphaBoolBlend {blendOption=Linear}, boneToModify {boneName=dyn_coat_fr_l_02}, scale=<0.1 0.1 0.1>}
   n19 = "Transform (Modify) Bone - Bone: dyn_coat_bk_02"(ComponentPose=@n18, Alpha=@n16, Translation="0,0,0", Rotation="0,0,0", Scale="0.1,0.1,0.1") {scaleMode=BMM_Replace, alphaBoolBlend {blendOption=Linear}, boneToModify {boneName=dyn_coat_bk_02}, scale=<0.1 0.1 0.1>}
   n20 = "Transform (Modify) Bone - Bone: dyn_coat_bk_l_02"(ComponentPose=@n19, Alpha=@n16, Translation="0,0,0", Rotation="0,0,0", Scale="0.1,0.1,0.1") {scaleMode=BMM_Replace, alphaBoolBlend {blendOption=Linear}, boneToModify {boneName=dyn_coat_bk_l_02}, scale=<0.1 0.1 0.1>}
   n21 = "Transform (Modify) Bone - Bone: dyn_coat_bk_r_02"(ComponentPose=@n20, Alpha=@n16, Translation="0,0,0", Rotation="0,0,0", Scale="0.1,0.1,0.1") {scaleMode=BMM_Replace, alphaBoolBlend {blendOption=Linear}, boneToModify {boneName=dyn_coat_bk_r_02}, scale=<0.1 0.1 0.1>}
   n22 = "Transform (Modify) Bone - Bone: dyn_coat_fr_r_02"(ComponentPose=@n21, Alpha=@n16, Translation="0,0,0", Rotation="0,0,0", Scale="0.1,0.1,0.1") {scaleMode=BMM_Replace, alphaBoolBlend {blendOption=Linear}, boneToModify {boneName=dyn_coat_fr_r_02}, scale=<0.1 0.1 0.1>}
   n23 = "Transform (Modify) Bone - Bone: dyn_coat_r_02"(ComponentPose=@n22, Alpha=@n16, Translation="0,0,0", Rotation="0,0,0", Scale="0.1,0.1,0.1") {scaleMode=BMM_Replace, alphaBoolBlend {blendOption=Linear}, boneToModify {boneName=dyn_coat_r_02}, scale=<0.1 0.1 0.1>}
   Save cached pose 'initial Pose'(Pose=Component To Local(ComponentPose=@n23)) {cacheName=initial Pose}
   n24 = RigidBody(ComponentPose=Local To Component(LocalPose=Use cached pose 'initial Pose'()), OverrideWorldGravity=@n8) {
    bClampLinearTranslationLimitToRefPose=true
    bOverrideWorldGravity=true
    overridePhysicsAsset=&/M_MED_Jonesy_Vagabond_Dyn_robe_skydiving
    simulationSpace=BaseBoneSpace
    alphaBoolBlend {blendOption=Linear}
    baseBoneRef {boneName=spine_02}
    componentAppliedLinearAccClamp=<3000 3000 3000>
    componentLinearAccScale=<0.4 0.3 0.3>
    componentLinearVelScale=<0.4 0.3 0.3>
    overrideWorldGravity {z=-980}
   }
   n25 = "Transform (Modify) Bone - Bone: dyn_coat_fr_l_01"(
    ComponentPose=Local To Component(LocalPose=Use cached pose 'initial Pose'())
    Alpha=Get thigh_rot_up_l() -> (thigh_rot_up_l:double)
    Translation="-3,0,6"
    Rotation="0,0.000,-60"
    Scale="1,1,1"
   ) {
    rotationMode=BMM_Additive
    rotationSpace=BCS_BoneSpace
    translationMode=BMM_Additive
    translationSpace=BCS_BoneSpace
    alphaBoolBlend {blendOption=Linear}
    boneToModify {boneName=dyn_coat_fr_l_01}
    rotation {roll=-60}
    translation {x=-3, z=6}
   }
   n26 = RigidBody(ComponentPose=@n25, Alpha=Get in vehicle rigid alpha() -> (in vehicle rigid alpha:double), OverrideWorldGravity=@n8) {
    bClampLinearTranslationLimitToRefPose=true
    bOverrideWorldGravity=true
    overridePhysicsAsset=&/M_MED_Jonesy_Vagabond_Dyn_robe
    simulationSpace=BaseBoneSpace
    alphaBoolBlend {blendOption=Linear}
    baseBoneRef {boneName=pelvis}
    componentAppliedLinearAccClamp=<3000 3000 3000>
    componentLinearAccScale=<0.4 0.3 0.3>
    componentLinearVelScale=<0.4 0.3 0.3>
    overrideWorldGravity {z=-980}
   }
   Save cached pose 'skydivingResult'(Pose=Blend Poses by bool(BlendPose_0=Component To Local(ComponentPose=@n24), BlendPose_1=Component To Local(ComponentPose=@n26), bActiveValue=Get bIsSkyDiving()) {blendType=Linear}) {cacheName=skydivingResult}
  ]}
 EventGraph {nodes [
   Event BlueprintUpdateAnimation()
   > n1 = Sequence!(execute) -> (then_0, then_1, then_2:Exec, then_3:Exec)
   > n2 = Branch(Condition=Get bIsSwimSprinting() -> (bIsSwimSprinting:bool))
   > Set gravity_override(gravity_override="0, 0, 0")
   n3 = Branch(execute=@n2.else, Condition=Get bIsSkyDiving())
   > n4 = Branch(Condition=Get bIsParachuteOpened())
   Set gravity_override(execute=[@n3.else, @n4.then], gravity_override="0, 0,-980")
   n5 = Branch(execute=@n4.else, Condition="float > float"(A=Get PawnVelocityZ(), B=50) {is=title K2Node_CallFunction})
   > Set gravity_override(gravity_override="0, 0,500")
   Set gravity_override(execute=@n5.else, gravity_override="0, 0,-300")
   n6 = Branch(execute=@n1.then_1, Condition=Get bIsOstrichVehicleDriver() -> (bIsOstrichVehicleDriver:bool))
   Set in vehicle rigid alpha(execute=[@n9.then, @n10.then, @n6.then], in vehicle rigid alpha:double=0) -> (Output_Get:double)
   > n7 = Set in vehicle scale alpha(in vehicle scale alpha:double=1) -> (Output_Get:double)
   Gate!(Enter:Exec=@n11.then, Open:Exec=@n7.then, Close:Exec=@n8.then, Toggle:Exec, bStartClosed:bool=true) -> (Exit:Exec) {is=title K2Node_MacroInstance}
   > Delay(Duration=0.3)
   > ResetDynamics(InTeleportType=TeleportPhysics)
   > n8 = ResetDynamics(InTeleportType=ResetPhysics)
   n9 = Branch(execute=@n6.else, Condition=Get bIsFerretVehicleDriver() -> (bIsFerretVehicleDriver:bool))
   n10 = Branch(execute=@n9.else, Condition=Get bIsOctopusVehicleDriver() -> (bIsOctopusVehicleDriver:bool))
   Set in vehicle rigid alpha(execute=@n10.else, in vehicle rigid alpha=1)
   > n11 = Set in vehicle scale alpha(in vehicle scale alpha=0)
   n12 = Break Fort Anim Input Slope Info!(FortAnimInput_SlopeInfo:"Fort Anim Input Slope Info Structure (by ref)"=Get SlopeInfoInput() -> (SlopeInfoInput:Fort Anim Input Slope Info Structure)) -> (RootSlopeTranslation:Vector=~, LeftFootSlopeTranslation:Vector=~, RightFootSlopeTranslation:Vector=~, SlopeRotation:Rotator=~) {is=title K2Node_BreakStruct}
   Set slope_foot_l_trans(execute=@n1.then_2, slope_foot_l_trans:Vector=@n12.LeftFootSlopeTranslation) -> (Output_Get:Vector)
   > Set slope_foot_r_trans(slope_foot_r_trans:Vector=@n12.RightFootSlopeTranslation) -> (Output_Get:Vector)
   > Set slope_foot_rot(slope_foot_rot:Rotator=@n12.SlopeRotation) -> (Output_Get:Rotator)
   n13 = BreakTransform(InTransform=GetDeltaTransformFromRefPose(self=GetOwnerSkeletalMeshComponent(), BoneName=thigh_l, BaseName=pelvis))
   Set thigh_rot_up_l(execute=@n1.then_3, thigh_rot_up_l:double=MapRangeClamped(Value=BreakRotator(InRot=@n13.Rotation).Yaw, InRangeA=-10, InRangeB=-100)) -> (Output_Get:double)
   Event OnBeginTargeting() {is=title K2Node_Event}
   > ResetDynamics(InTeleportType=TeleportPhysics)
   > ResetDynamics(InTeleportType=ResetPhysics)
   Event OnParachuteOpened()
   > Delay(Duration=1.5)
   Event OnEndSkydiving()
   > Delay(Duration=0.3)
   Event OnBeginSkydiving()
  ]}
]
variables [
 gravity_override {type=Vector}
 in vehicle rigid alpha {type=number}
 in vehicle scale alpha {type=number}
 slope_foot_l_trans {type=Vector}
 slope_foot_r_trans {type=Vector}
 slope_foot_rot {type=Rotator}
 thigh_rot_up_l {type=number}
]
