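# Physics asset for the Jonesy Vagabond male player body, setting up dynamic hair and beard motion.
# Seven four-bone hair chains hang from the head with limited swing and tight twist, while three short beard chains attach through free-rotating master bodies.
# Free-angular cross-constraints tie neighboring hair chains at their second and fourth bones and link the beard tips.
# Head, jaw, neck, upper spine and clavicles supply surrounding body shapes.
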PhysicsAsset /BRCosmetics/Characters/Player/Male/Medium/Bodies/M_MED_Jonesy_Vagabond/Meshes/Parts/M_MED_Jonesy_Vagabond_Dyn
bodies {
 C_jaw [{shapeType=Capsule, length=0.5, radius=0.5, rotation=<rot 90 0 0>}]
 clavicle_l [{shapeType=Capsule, length=8.1875, radius=10.5, center=<9.1717 0.0003 -1.2623>, rotation=<rot 76.2574 0.3809 0.3333>}]
 clavicle_r [{shapeType=Capsule, length=9.0625, radius=9.625, center=<-8.6127 0.9109 0.1634>, rotation=<rot 76.2566 0.3796 0.3319>}]
 dyn_beard_01 [{shapeType=Capsule, length=1.5938, radius=2.6562, center=<2.0895 0 0>, rotation=<rot 79.9999 -180 180>}]
 dyn_beard_01_MSTR [{shapeType=Capsule, length=0.5, radius=0.5, rotation=<rot 90 0 0>}]
 dyn_beard_02 [{shapeType=Capsule, length=1.8125, radius=2.9375, center=<3.367 0 0.7>, rotation=<rot 79.9997 -180 180>}]
 dyn_beard_l_01 [{shapeType=Capsule, length=2.2188, radius=2.0312, center=<2.0895 0 1>, rotation=<rot 85 90.0014 -179.9995>}]
 dyn_beard_l_01_MSTR [{shapeType=Capsule, length=0.5, radius=0.5, rotation=<rot 90 0 0>}]
 dyn_beard_l_02 [{shapeType=Capsule, length=1.8438, radius=2.4062, center=<3 0 1>, rotation=<rot 1.2927 -14.9453 94.8304>}]
 dyn_beard_r_01 [{shapeType=Capsule, length=2.2188, radius=2.0312, center=<2.0895 0 1>, rotation=<rot 0.0001 -0.0001 90>}]
 dyn_beard_r_01_MSTR [{shapeType=Capsule, length=0.5, radius=0.5, rotation=<rot 90 0 0>}]
 dyn_beard_r_02 [{shapeType=Capsule, length=1.7188, radius=2.5312, center=<3 0 1>, rotation=<rot 1.2996 9.906 80.0937>}]
 dyn_hair_bk_01 [{shapeType=Capsule, length=5.6562, radius=3.7, center=<2.3259 0.0004 3>, rotation=<rot 90 0 0>}]
 dyn_hair_bk_02 [{shapeType=Capsule, length=5.6562, radius=3.7, center=<2.3259 0.0004 3>, rotation=<rot 0 0 90>}]
 dyn_hair_bk_03 [{shapeType=Capsule, length=3.0938, radius=3.575, center=<2.3259 0 3>, rotation=<rot 0 0 90>}]
 dyn_hair_bk_04 [{shapeType=Capsule, length=4.5, radius=2.5, shapeName=dyn_hair_bk_04_capsule, center=<2.9216 0 2>, rotation=<rot 0 -0.0001 -90.0003>}]
 dyn_hair_bk_l_01 [{shapeType=Capsule, length=5.6562, radius=3.7, center=<2.3259 0.0004 3>, rotation=<rot 90 0 0>}]
 dyn_hair_bk_l_02 [{shapeType=Capsule, length=5.6562, radius=3.7, center=<2.3259 0.0004 3>, rotation=<rot 0 0 90>}]
 dyn_hair_bk_l_03 [{shapeType=Capsule, length=2.9688, radius=3.7, center=<2.3259 0 3>, rotation=<rot 0 0 90>}]
 dyn_hair_bk_l_04 [{shapeType=Capsule, length=4.5, radius=2.5, shapeName=dyn_hair_bk_l_04_capsule, center=<2.9216 0 2>, rotation=<rot 0 0 -95.0004>}]
 dyn_hair_bk_r_01 [{shapeType=Capsule, length=5.6562, radius=3.7, center=<2.3259 0.0004 3>, rotation=<rot 90 0 0>}]
 dyn_hair_bk_r_02 [{shapeType=Capsule, length=5.6562, radius=3.7, center=<2.3259 0.0004 3>, rotation=<rot 0 0 90>}]
 dyn_hair_bk_r_03 [{shapeType=Capsule, length=2.9688, radius=3.7, center=<2.3259 0 3>, rotation=<rot 0 0 90>}]
 dyn_hair_bk_r_04 [{
   shapeType=Capsule
   length=4.25
   radius=2.75
   shapeName=dyn_hair_bk_r_04_capsule
   center=<2.9217 0 2>
   rotation=<rot -0.0003 -0.0003 -80.0009>
  }]
 dyn_hair_fr_l_01 [{shapeType=Capsule, length=5.6562, radius=3.7, center=<2.3259 0.0004 3>, rotation=<rot 90 0 0>}]
 dyn_hair_fr_l_02 [{shapeType=Capsule, length=5.6562, radius=3.7, center=<2.3259 0.0004 3>, rotation=<rot 90 0 0>}]
 dyn_hair_fr_l_03 [{shapeType=Capsule, length=2.9688, radius=3.7, center=<2.3259 0.0004 3>, rotation=<rot 90 0 0>}]
 dyn_hair_fr_l_04 [{
   shapeType=Capsule
   length=3.75
   radius=3.25
   shapeName=dyn_hair_fr_l_04_capsule
   center=<2.9216 0.0001 2>
   rotation=<rot -0.0004 -25 -90.001>
  }]
 dyn_hair_fr_r_01 [{shapeType=Capsule, length=5.6562, radius=3.7, center=<2.3259 0.0004 3>, rotation=<rot 90 0 0>}]
 dyn_hair_fr_r_02 [{shapeType=Capsule, length=5.6562, radius=3.7, center=<2.3259 0.0004 3>, rotation=<rot 90 0 0>}]
 dyn_hair_fr_r_03 [{shapeType=Capsule, length=2.9688, radius=3.7, center=<2.3259 0.0004 3>, rotation=<rot 0 0 90>}]
 dyn_hair_fr_r_04 [{
   shapeType=Capsule
   length=4.75
   radius=2.25
   shapeName=dyn_hair_fr_r_04_capsule
   center=<2.9216 0 1>
   rotation=<rot -0.0001 14.9997 -89.9996>
  }]
 dyn_hair_l_01 [{shapeType=Capsule, length=5.6562, radius=3.7, center=<2.3259 0.0004 3>, rotation=<rot 90 0 0>}]
 dyn_hair_l_02 [{shapeType=Capsule, length=5.6562, radius=3.7, center=<2.3259 0.0004 3>, rotation=<rot 90 0 0>}]
 dyn_hair_l_03 [{shapeType=Capsule, length=2.9688, radius=3.7, center=<2.3259 0.0004 3>, rotation=<rot 90 0 0>}]
 dyn_hair_l_04 [{shapeType=Capsule, length=4.75, radius=2.25, shapeName=dyn_hair_l_04_capsule, center=<2 0.0001 2>, rotation=<rot 0 -0.0001 -90.0003>}]
 dyn_hair_r_01 [{shapeType=Capsule, length=5.6562, radius=3.7, center=<2.3259 0.0004 3>, rotation=<rot 90 0 0>}]
 dyn_hair_r_02 [{shapeType=Capsule, length=5.6562, radius=3.7, center=<2.3259 0.0004 3>, rotation=<rot 90 0 0>}]
 dyn_hair_r_03 [{shapeType=Capsule, length=2.9688, radius=3.7, center=<2.3259 0.0004 3>, rotation=<rot 0 0 90>}]
 dyn_hair_r_04 [{shapeType=Capsule, length=4.75, radius=2.25, shapeName=dyn_hair_r_04_capsule, center=<2 0 3>, rotation=<rot 0 -0.0001 -90.0011>}]
 head [
  {shapeType=Sphere, radius=6.1875, center=<4.2242 -0.3879 -0.0379>}
  {shapeType=Capsule, length=7.2812, radius=6.6562, center=<-0.2526 3.6745 0>, rotation=<rot 90 0 0>}
 ]
 neck_02 [{shapeType=Capsule, length=3, radius=7, center=<-2.6062 -0.9293 0>, rotation=<rot 90 180 176.5972>}]
 spine_05 [
  {shapeType=Capsule, length=4.3438, radius=8.2188, center=<9.7533 -0.0001 0.0001>}
  {shapeType=Capsule, length=13.3438, radius=9, center=<0.5156 2.7007 0>}
  {shapeType=Capsule, length=11.25, radius=14.5, center=<-0.7066 -2.1772 0>}
 ]
}
constraints [
 {bone1Name=dyn_hair_bk_02, bone2Name=dyn_hair_bk_01, swing1LimitDegrees=15, swing2LimitDegrees=45, twistLimitDegrees=5}
 {bone1Name=dyn_hair_bk_03, bone2Name=dyn_hair_bk_02, swing1LimitDegrees=15, swing2LimitDegrees=45, twistLimitDegrees=5}
 {bone1Name=dyn_hair_bk_04, bone2Name=dyn_hair_bk_03, swing1LimitDegrees=15, swing2LimitDegrees=45, twistLimitDegrees=5}
 {bone1Name=dyn_hair_bk_l_02, bone2Name=dyn_hair_bk_l_01, swing1LimitDegrees=15, swing2LimitDegrees=45, twistLimitDegrees=5}
 {bone1Name=dyn_hair_bk_l_03, bone2Name=dyn_hair_bk_l_02, swing1LimitDegrees=15, swing2LimitDegrees=45, twistLimitDegrees=5}
 {bone1Name=dyn_hair_bk_l_04, bone2Name=dyn_hair_bk_l_03, swing1LimitDegrees=15, swing2LimitDegrees=45, twistLimitDegrees=5}
 {bone1Name=dyn_hair_l_02, bone2Name=dyn_hair_l_01, swing1LimitDegrees=15, swing2LimitDegrees=45, twistLimitDegrees=5}
 {bone1Name=dyn_hair_l_03, bone2Name=dyn_hair_l_02, swing1LimitDegrees=15, swing2LimitDegrees=45, twistLimitDegrees=5}
 {bone1Name=dyn_hair_l_04, bone2Name=dyn_hair_l_03, swing1LimitDegrees=15, swing2LimitDegrees=45, twistLimitDegrees=5}
 {bone1Name=dyn_hair_fr_l_02, bone2Name=dyn_hair_fr_l_01, swing1LimitDegrees=15, swing2LimitDegrees=45, twistLimitDegrees=5}
 {bone1Name=dyn_hair_fr_l_03, bone2Name=dyn_hair_fr_l_02, swing1LimitDegrees=15, swing2LimitDegrees=45, twistLimitDegrees=5}
 {bone1Name=dyn_hair_fr_l_04, bone2Name=dyn_hair_fr_l_03, swing1LimitDegrees=15, swing2LimitDegrees=45, twistLimitDegrees=5}
 {bone1Name=dyn_beard_01, bone2Name=dyn_beard_01_MSTR, swing1LimitDegrees=30, swing2LimitDegrees=30, twistLimitDegrees=5}
 {bone1Name=dyn_beard_02, bone2Name=dyn_beard_01, swing1LimitDegrees=30, swing2LimitDegrees=30, twistLimitDegrees=5}
 {bone1Name=dyn_beard_l_01, bone2Name=dyn_beard_l_01_MSTR, swing1LimitDegrees=30, swing2LimitDegrees=30, twistLimitDegrees=5}
 {bone1Name=dyn_beard_l_02, bone2Name=dyn_beard_l_01, swing1LimitDegrees=30, swing2LimitDegrees=30, twistLimitDegrees=5}
 {bone1Name=dyn_beard_r_01, bone2Name=dyn_beard_r_01_MSTR, swing1LimitDegrees=30, swing2LimitDegrees=30, twistLimitDegrees=5}
 {bone1Name=dyn_beard_r_02, bone2Name=dyn_beard_r_01, swing1LimitDegrees=30, swing2LimitDegrees=30, twistLimitDegrees=5}
 {bone1Name=dyn_hair_fr_r_02, bone2Name=dyn_hair_fr_r_01, swing1LimitDegrees=15, swing2LimitDegrees=45, twistLimitDegrees=5}
 {bone1Name=dyn_hair_fr_r_03, bone2Name=dyn_hair_fr_r_02, swing1LimitDegrees=15, swing2LimitDegrees=45, twistLimitDegrees=5}
 {bone1Name=dyn_hair_fr_r_04, bone2Name=dyn_hair_fr_r_03, swing1LimitDegrees=15, swing2LimitDegrees=45, twistLimitDegrees=5}
 {bone1Name=dyn_hair_r_02, bone2Name=dyn_hair_r_01, swing1LimitDegrees=25, swing2LimitDegrees=45, twistLimitDegrees=5}
 {bone1Name=dyn_hair_r_03, bone2Name=dyn_hair_r_02, swing1LimitDegrees=15, swing2LimitDegrees=45, twistLimitDegrees=5}
 {bone1Name=dyn_hair_r_04, bone2Name=dyn_hair_r_03, swing1LimitDegrees=15, swing2LimitDegrees=45, twistLimitDegrees=5}
 {bone1Name=dyn_hair_bk_r_02, bone2Name=dyn_hair_bk_r_01, swing1LimitDegrees=15, swing2LimitDegrees=45, twistLimitDegrees=5}
 {bone1Name=dyn_hair_bk_r_03, bone2Name=dyn_hair_bk_r_02, swing1LimitDegrees=15, swing2LimitDegrees=45, twistLimitDegrees=5}
 {bone1Name=dyn_hair_bk_r_04, bone2Name=dyn_hair_bk_r_03, swing1LimitDegrees=15, swing2LimitDegrees=45, twistLimitDegrees=5}
 {bone1Name=dyn_hair_bk_01, bone2Name=head, swing1LimitDegrees=15, swing2LimitDegrees=45, twistLimitDegrees=5}
 {bone1Name=dyn_hair_bk_l_01, bone2Name=head, swing1LimitDegrees=15, swing2LimitDegrees=45, twistLimitDegrees=5}
 {bone1Name=dyn_hair_l_01, bone2Name=head, swing1LimitDegrees=15, swing2LimitDegrees=45, twistLimitDegrees=5}
 {bone1Name=dyn_hair_fr_l_01, bone2Name=head, swing1LimitDegrees=15, swing2LimitDegrees=45, twistLimitDegrees=5}
 {
  bone1Name=dyn_beard_01_MSTR
  bone2Name=head
  swing1LimitDegrees=45
  swing1Motion=Free
  swing2LimitDegrees=45
  swing2Motion=Free
  twistLimitDegrees=45
  twistMotion=Free
 }
 {
  bone1Name=dyn_beard_l_01_MSTR
  bone2Name=head
  swing1LimitDegrees=45
  swing1Motion=Free
  swing2LimitDegrees=45
  swing2Motion=Free
  twistLimitDegrees=45
  twistMotion=Free
 }
 {
  bone1Name=dyn_beard_r_01_MSTR
  bone2Name=head
  swing1LimitDegrees=45
  swing1Motion=Free
  swing2LimitDegrees=45
  swing2Motion=Free
  twistLimitDegrees=45
  twistMotion=Free
 }
 {bone1Name=dyn_hair_fr_r_01, bone2Name=head, swing1LimitDegrees=15, swing2LimitDegrees=45, twistLimitDegrees=5}
 {bone1Name=dyn_hair_r_01, bone2Name=head, swing1LimitDegrees=15, swing2LimitDegrees=45, twistLimitDegrees=5}
 {bone1Name=dyn_hair_bk_r_01, bone2Name=head, swing1LimitDegrees=15, swing2LimitDegrees=45, twistLimitDegrees=5}
 {
  bone1Name=head
  bone2Name=neck_02
  swing1LimitDegrees=45
  swing1Motion=Free
  swing2LimitDegrees=45
  swing2Motion=Free
  twistLimitDegrees=45
  twistMotion=Free
 }
 {
  bone1Name=clavicle_l
  bone2Name=spine_05
  swing1LimitDegrees=45
  swing1Motion=Free
  swing2LimitDegrees=45
  swing2Motion=Free
  twistLimitDegrees=45
  twistMotion=Free
 }
 {
  bone1Name=clavicle_r
  bone2Name=spine_05
  swing1LimitDegrees=45
  swing1Motion=Free
  swing2LimitDegrees=45
  swing2Motion=Free
  twistLimitDegrees=45
  twistMotion=Free
 }
 {
  bone1Name=neck_02
  bone2Name=spine_05
  swing1LimitDegrees=45
  swing1Motion=Free
  swing2LimitDegrees=45
  swing2Motion=Free
  twistLimitDegrees=45
  twistMotion=Free
 }
 {
  bone1Name=dyn_hair_fr_r_02
  bone2Name=dyn_hair_r_02
  swing1LimitDegrees=45
  swing1Motion=Free
  swing2LimitDegrees=45
  swing2Motion=Free
  twistLimitDegrees=45
  twistMotion=Free
 }
 {
  bone1Name=dyn_hair_fr_r_04
  bone2Name=dyn_hair_r_04
  swing1LimitDegrees=45
  swing1Motion=Free
  swing2LimitDegrees=45
  swing2Motion=Free
  twistLimitDegrees=45
  twistMotion=Free
 }
 {
  bone1Name=dyn_hair_r_02
  bone2Name=dyn_hair_bk_r_02
  swing1LimitDegrees=45
  swing1Motion=Free
  swing2LimitDegrees=45
  swing2Motion=Free
  twistLimitDegrees=45
  twistMotion=Free
 }
 {
  bone1Name=dyn_hair_r_04
  bone2Name=dyn_hair_bk_r_04
  swing1LimitDegrees=45
  swing1Motion=Free
  swing2LimitDegrees=45
  swing2Motion=Free
  twistLimitDegrees=45
  twistMotion=Free
 }
 {
  bone1Name=dyn_hair_fr_l_02
  bone2Name=dyn_hair_l_02
  swing1LimitDegrees=45
  swing1Motion=Free
  swing2LimitDegrees=45
  swing2Motion=Free
  twistLimitDegrees=45
  twistMotion=Free
 }
 {
  bone1Name=dyn_hair_fr_l_04
  bone2Name=dyn_hair_l_04
  swing1LimitDegrees=45
  swing1Motion=Free
  swing2LimitDegrees=45
  swing2Motion=Free
  twistLimitDegrees=45
  twistMotion=Free
 }
 {
  bone1Name=dyn_hair_l_02
  bone2Name=dyn_hair_bk_l_02
  swing1LimitDegrees=45
  swing1Motion=Free
  swing2LimitDegrees=45
  swing2Motion=Free
  twistLimitDegrees=45
  twistMotion=Free
 }
 {
  bone1Name=dyn_hair_l_04
  bone2Name=dyn_hair_bk_l_04
  swing1LimitDegrees=45
  swing1Motion=Free
  swing2LimitDegrees=45
  swing2Motion=Free
  twistLimitDegrees=45
  twistMotion=Free
 }
 {
  bone1Name=dyn_hair_bk_02
  bone2Name=dyn_hair_bk_l_02
  swing1LimitDegrees=45
  swing1Motion=Free
  swing2LimitDegrees=45
  swing2Motion=Free
  twistLimitDegrees=45
  twistMotion=Free
 }
 {
  bone1Name=dyn_hair_bk_02
  bone2Name=dyn_hair_bk_r_02
  swing1LimitDegrees=45
  swing1Motion=Free
  swing2LimitDegrees=45
  swing2Motion=Free
  twistLimitDegrees=45
  twistMotion=Free
 }
 {
  bone1Name=dyn_hair_bk_04
  bone2Name=dyn_hair_bk_l_04
  swing1LimitDegrees=45
  swing1Motion=Free
  swing2LimitDegrees=45
  swing2Motion=Free
  twistLimitDegrees=45
  twistMotion=Free
 }
 {
  bone1Name=dyn_hair_bk_04
  bone2Name=dyn_hair_bk_r_04
  swing1LimitDegrees=45
  swing1Motion=Free
  swing2LimitDegrees=45
  swing2Motion=Free
  twistLimitDegrees=45
  twistMotion=Free
 }
 {
  bone1Name=C_jaw
  bone2Name=head
  swing1LimitDegrees=45
  swing1Motion=Free
  swing2LimitDegrees=45
  swing2Motion=Free
  twistLimitDegrees=45
  twistMotion=Free
 }
 {
  bone1Name=dyn_beard_r_02
  bone2Name=dyn_beard_02
  swing1LimitDegrees=45
  swing1Motion=Free
  swing2LimitDegrees=45
  swing2Motion=Free
  twistLimitDegrees=45
  twistMotion=Free
 }
 {
  bone1Name=dyn_beard_l_02
  bone2Name=dyn_beard_02
  swing1LimitDegrees=45
  swing1Motion=Free
  swing2LimitDegrees=45
  swing2Motion=Free
  twistLimitDegrees=45
  twistMotion=Free
 }
]
