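# Animation Blueprint for the Jonesy Vagabond character part, copying the parent mesh pose and driving beard, hair and bow bones.
# Event graph maps head rotation against spine_05 to alphas for additive beard and hair corrections, and detects a RoyalWonk body, reading its bow transforms.
# Two neck-relative RigidBody simulations, one with a RoyalWonk-specific physics asset, are blended by bool into the output.
# A mustache and beard constraint chain exists but is not connected to the output.
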AnimBlueprint /BRCosmetics/Characters/Player/Male/Medium/Bodies/M_MED_Jonesy_Vagabond/Meshes/Parts/M_MED_Jonesy_Vagabond_AnimBP
target_skeleton=&/M_MED_Jonesy_Vagabond_Skeleton
graphs [
 AnimGraph {nodes [
   n1 = "Constraint - dyn_mustache_l_01_MSTR (PPPPPP)"!(
    ComponentPose
    Alpha=1
    ConstraintWeights_0=0.2367
    ConstraintWeights_1:float=0.0736
    ConstraintWeights_2:float=0.3961
    ConstraintWeights_3:float=0.2298
    ConstraintWeights_4:float=0.0445
    ConstraintWeights_5:float=0.0193
   ) -> (Pose) {
    alphaBoolBlend {blendOption=Linear}
    boneToModify {boneName=dyn_mustache_l_01_MSTR}
    constraintSetup [
     {targetBone {boneName=head}}
     {targetBone {boneName=L_lip_corner}}
     {targetBone {boneName=C_lip_upper_mid}}
     {targetBone {boneName=L_lip_upper_outer}}
     {targetBone {boneName=C_jaw}}
     {targetBone {boneName=L_cheek_inner}}
    ]
    constraintWeights [1,1,1,1,1,1]
   }
   n2 = "Constraint - dyn_mustache_r_01_MSTR (PPPPPP)"!(
    ComponentPose=@n1
    Alpha=1
    ConstraintWeights_0=0.2367
    ConstraintWeights_1:float=0.2298
    ConstraintWeights_2:float=0.3961
    ConstraintWeights_3:float=0.0736
    ConstraintWeights_4:float=0.0445
    ConstraintWeights_5:float=0.0193
   ) -> (Pose) {
    alphaBoolBlend {blendOption=Linear}
    boneToModify {boneName=dyn_mustache_r_01_MSTR}
    constraintSetup [
     {targetBone {boneName=head}}
     {targetBone {boneName=R_lip_upper_outer}}
     {targetBone {boneName=C_lip_upper_mid}}
     {targetBone {boneName=R_lip_corner}}
     {targetBone {boneName=C_jaw}}
     {targetBone {boneName=R_cheek_inner}}
    ]
    constraintWeights [1,1,1,1,1,1]
   }
   n3 = "Constraint - dyn_beard_r_01_MSTR (PP)"!(ComponentPose=@n2, Alpha=1, ConstraintWeights_0=0.8344, ConstraintWeights_1:float=0.1656) -> (Pose) {
    alphaBoolBlend {blendOption=Linear}
    boneToModify {boneName=dyn_beard_r_01_MSTR}
    constraintSetup [{targetBone {boneName=C_jaw}}, {targetBone {boneName=head}}]
    constraintWeights [1,1]
   }
   n4 = "Constraint - dyn_beard_01_MSTR (P)"!(ComponentPose=@n3, Alpha=1, ConstraintWeights_0=1, ConstraintWeights_1:float=0) -> (Pose) {
    alphaBoolBlend {blendOption=Linear}
    boneToModify {boneName=dyn_beard_01_MSTR}
    constraintSetup [{targetBone {boneName=C_jaw}}]
    constraintWeights [1,0]
   }
   "Constraint - dyn_beard_l_01_MSTR (PP)"!(ComponentPose=@n4, Alpha=1, ConstraintWeights_0=0.8344, ConstraintWeights_1:float=0.1656) -> (Pose) {
    alphaBoolBlend {blendOption=Linear}
    boneToModify {boneName=dyn_beard_l_01_MSTR}
    constraintSetup [{targetBone {boneName=C_jaw}}, {targetBone {boneName=head}}]
    constraintWeights [1,1]
   }
   Output Pose(Result=Use cached pose 'dyn_cache'()) {layerGroup=None}
   n5 = "Transform (Modify) Bone - Bone: base_bow"!(
    ComponentPose=Local To Component(LocalPose=Copy Pose From Mesh() {bUseAttachedParent=true})
    bAlphaBoolEnabled:bool=True
    Translation="0,0,0"
    Rotation="0,0,0"
   ) -> (Pose) {
    alphaInputType=Bool
    rotationMode=BMM_Replace
    rotationSpace=BCS_ParentBoneSpace
    translationMode=BMM_Replace
    translationSpace=BCS_ParentBoneSpace
    alphaBoolBlend {blendOption=Linear}
    boneToModify {boneName=base_bow}
   }
   n6 = "Transform (Modify) Bone - Bone: dyn_bow"!(ComponentPose=@n5, bAlphaBoolEnabled:bool=True, Translation="0,0,0", Rotation="0,0,0") -> (Pose) {
    alphaInputType=Bool
    rotationMode=BMM_Replace
    rotationSpace=BCS_ParentBoneSpace
    translationMode=BMM_Replace
    translationSpace=BCS_ParentBoneSpace
    alphaBoolBlend {blendOption=Linear}
    boneToModify {boneName=dyn_bow}
   }
   Save cached pose 'init_cache'(Pose=Component To Local(ComponentPose=@n6)) {cacheName=init_cache}
   n7 = Get head_rot_Fr() -> (head_rot_Fr:double)
   n8 = "Transform (Modify) Bone - Bone: dyn_beard_01"!(ComponentPose=Local To Component(LocalPose=Use cached pose 'init_cache'()), Alpha=@n7, Translation="-1,0,2", Rotation="20,0,0") -> (Pose) {
    rotationMode=BMM_Additive
    rotationSpace=BCS_BoneSpace
    translationMode=BMM_Additive
    translationSpace=BCS_BoneSpace
    alphaBoolBlend {blendOption=Linear}
    boneToModify {boneName=dyn_beard_01}
    rotation {pitch=20}
    translation {x=-1, z=2}
   }
   n9 = "Transform (Modify) Bone - Bone: dyn_beard_r_02"!(ComponentPose=@n8, Alpha=@n7, Translation="-1,0,2", Rotation="40,0,0") -> (Pose) {
    rotationMode=BMM_Additive
    rotationSpace=BCS_BoneSpace
    translationMode=BMM_Additive
    translationSpace=BCS_BoneSpace
    alphaBoolBlend {blendOption=Linear}
    boneToModify {boneName=dyn_beard_r_02}
    rotation {pitch=40}
    translation {x=-1, z=2}
   }
   n10 = "Transform (Modify) Bone - Bone: dyn_beard_l_01"!(ComponentPose=@n9, Alpha=@n7, Translation="-1,0,2", Rotation="20,0,0") -> (Pose) {
    rotationMode=BMM_Additive
    rotationSpace=BCS_BoneSpace
    translationMode=BMM_Additive
    translationSpace=BCS_BoneSpace
    alphaBoolBlend {blendOption=Linear}
    boneToModify {boneName=dyn_beard_l_01}
    rotation {pitch=20}
    translation {x=-1, z=2}
   }
   n11 = "Transform (Modify) Bone - Bone: dyn_beard_01"(ComponentPose=@n10, Alpha=Get lookingDownExtreme() -> (lookingDownExtreme:double), Translation="-1,-1,4", Rotation="20,0,0") {
    rotationMode=BMM_Additive
    rotationSpace=BCS_BoneSpace
    translationMode=BMM_Additive
    translationSpace=BCS_BoneSpace
    alphaBoolBlend {blendOption=Linear}
    boneToModify {boneName=dyn_beard_01}
    rotation {pitch=20}
    translation=<-1 -1 4>
   }
   n12 = Get headRotDownLeft() -> (headRotDownLeft:double)
   n13 = "Transform (Modify) Bone - Bone: dyn_beard_l_01"(ComponentPose=@n11, Alpha=@n12, Translation="-1,3,0.5", Rotation="10,40,0") {
    rotationMode=BMM_Additive
    rotationSpace=BCS_BoneSpace
    translationMode=BMM_Additive
    translationSpace=BCS_BoneSpace
    alphaBoolBlend {blendOption=Linear}
    boneToModify {boneName=dyn_beard_l_01}
    rotation {pitch=10, yaw=40}
    translation=<-1 3 0.5>
   }
   n14 = "Transform (Modify) Bone - Bone: dyn_beard_01"(ComponentPose=@n13, Alpha=@n12, Translation="-1,1.5,0.5", Rotation="5,20,0") {
    rotationMode=BMM_Additive
    rotationSpace=BCS_BoneSpace
    translationMode=BMM_Additive
    translationSpace=BCS_BoneSpace
    alphaBoolBlend {blendOption=Linear}
    boneToModify {boneName=dyn_beard_01}
    rotation {pitch=5, yaw=20}
    translation=<-1 1.5 0.5>
   }
   n15 = Get headRotDownRight() -> (headRotDownRight:double)
   n16 = "Transform (Modify) Bone - Bone: dyn_beard_r_01"!(ComponentPose=@n14, Alpha=@n15, Translation="-1,-3,0.5", Rotation="10,-40,0") -> (Pose) {
    rotationMode=BMM_Additive
    rotationSpace=BCS_BoneSpace
    translationMode=BMM_Additive
    translationSpace=BCS_BoneSpace
    alphaBoolBlend {blendOption=Linear}
    boneToModify {boneName=dyn_beard_r_01}
    rotation {pitch=10, yaw=-40}
    translation=<-1 -3 0.5>
   }
   n17 = "Transform (Modify) Bone - Bone: dyn_beard_01"(ComponentPose=@n16, Alpha=@n15, Translation="-1,-1.5,-0.5", Rotation="5,-20,0") {
    rotationMode=BMM_Additive
    rotationSpace=BCS_BoneSpace
    translationMode=BMM_Additive
    translationSpace=BCS_BoneSpace
    alphaBoolBlend {blendOption=Linear}
    boneToModify {boneName=dyn_beard_01}
    rotation {pitch=5, yaw=-20}
    translation=<-1 -1.5 -0.5>
   }
   n18 = Component To Local(ComponentPose=@n17)
   n19 = Use cached pose 'init_cache'()
   n20 = Get head_rot_bk() -> (head_rot_bk:double)
   n21 = "Transform (Modify) Bone - Bone: dyn_hair_bk_02"(
    ComponentPose=Local To Component(LocalPose=Blend Poses by bool(BlendPose_0=@n18, BlendPose_1=@n19))
    Alpha=@n20
    Translation="-2,0,2"
    Rotation="50,0.000,-0.000"
    Scale="1,1,1"
   ) {
    rotationMode=BMM_Additive
    rotationSpace=BCS_BoneSpace
    translationMode=BMM_Additive
    translationSpace=BCS_BoneSpace
    alphaBoolBlend {blendOption=Linear}
    boneToModify {boneName=dyn_hair_bk_02}
    rotation {pitch=50}
    translation {x=-2, z=2}
   }
   n22 = "Transform (Modify) Bone - Bone: dyn_hair_bk_l_02"(ComponentPose=@n21, Alpha=@n20, Translation="-2,0,2", Rotation="40,-35,0", Scale="1,1,1") {
    rotationMode=BMM_Additive
    rotationSpace=BCS_BoneSpace
    translationMode=BMM_Additive
    translationSpace=BCS_BoneSpace
    alphaBoolBlend {blendOption=Linear}
    boneToModify {boneName=dyn_hair_bk_l_02}
    rotation {pitch=40, yaw=-35}
    translation {x=-2, z=2}
   }
   n23 = "Transform (Modify) Bone - Bone: dyn_hair_bk_r_02"(ComponentPose=@n22, Alpha=@n20, Translation="-2,0,2", Rotation="40,15,-0.000", Scale="1,1,1") {
    rotationMode=BMM_Additive
    rotationSpace=BCS_BoneSpace
    translationMode=BMM_Additive
    translationSpace=BCS_BoneSpace
    alphaBoolBlend {blendOption=Linear}
    boneToModify {boneName=dyn_hair_bk_r_02}
    rotation {pitch=40, yaw=15}
    translation {x=-2, z=2}
   }
   n24 = Get head_rot_r() -> (head_rot_r:double)
   n25 = "Transform (Modify) Bone - Bone: dyn_hair_r_02"(ComponentPose=@n23, Alpha=@n24, Translation="-3,0,2", Rotation="30,30,0", Scale="1,1,1") {
    rotationMode=BMM_Additive
    rotationSpace=BCS_BoneSpace
    translationMode=BMM_Additive
    translationSpace=BCS_BoneSpace
    alphaBoolBlend {blendOption=Linear}
    boneToModify {boneName=dyn_hair_r_02}
    rotation {pitch=30, yaw=30}
    translation {x=-3, z=2}
   }
   n26 = "Transform (Modify) Bone - Bone: dyn_hair_fr_r_02"(ComponentPose=@n25, Alpha=@n24, Translation="-3,0,0", Rotation="0,0,0", Scale="1,1,1") {
    rotationMode=BMM_Additive
    rotationSpace=BCS_BoneSpace
    translationMode=BMM_Additive
    translationSpace=BCS_BoneSpace
    alphaBoolBlend {blendOption=Linear}
    boneToModify {boneName=dyn_hair_fr_r_02}
    translation {x=-3}
   }
   Save cached pose 'preDyn_cache'(Pose=Component To Local(ComponentPose=@n26)) {cacheName=preDyn_cache}
   n27 = RigidBody(ComponentPose=Local To Component(LocalPose=Use cached pose 'preDyn_cache'())) {
    bClampLinearTranslationLimitToRefPose=true
    bOverrideWorldGravity=true
    overridePhysicsAsset=&/M_MED_Jonesy_Vagabond_RoyalWonk_Dyn
    simulationSpace=BaseBoneSpace
    alphaBoolBlend {blendOption=Linear}
    baseBoneRef {boneName=neck_01}
    componentAppliedLinearAccClamp=<3000 3000 3000>
    componentLinearAccScale=<0.4 0.3 0.3>
    componentLinearVelScale=<0.4 0.3 0.3>
    overrideWorldGravity {z=-980}
   }
   n28 = RigidBody(ComponentPose=Local To Component(LocalPose=Use cached pose 'preDyn_cache'())) {
    bClampLinearTranslationLimitToRefPose=true
    bOverrideWorldGravity=true
    overridePhysicsAsset=&/M_MED_Jonesy_Vagabond_Dyn
    simulationSpace=BaseBoneSpace
    alphaBoolBlend {blendOption=Linear}
    baseBoneRef {boneName=neck_01}
    componentAppliedLinearAccClamp=<3000 3000 3000>
    componentLinearAccScale=<0.4 0.3 0.3>
    componentLinearVelScale=<0.4 0.3 0.3>
    overrideWorldGravity {z=-980}
   }
   Save cached pose 'dyn_cache'(Pose=Blend Poses by bool(BlendPose_0=Component To Local(ComponentPose=@n27), BlendPose_1=Component To Local(ComponentPose=@n28))) {cacheName=dyn_cache}
  ]}
 EventGraph {nodes [
   Event BlueprintUpdateAnimation()
   > n1 = Sequence!(execute) -> (then_0, then_1, then_2:Exec, then_3:Exec)
   n2 = BreakTransform(InTransform=GetDeltaTransformFromRefPose(self=GetOwnerSkeletalMeshComponent(), BoneName=head, BaseName=spine_05))
   n3 = BreakRotator(InRot=@n2.Rotation)
   > Set head_rot_bk(head_rot_bk:double=MapRangeClamped(Value=@n3.Yaw, InRangeA=0, InRangeB=-50)) -> (Output_Get:double)
   n4 = MapRangeClamped(Value=@n3.Yaw, InRangeA=-10, InRangeB=15)
   > Set head_rot_Fr(head_rot_Fr:double=@n4) -> (Output_Get:double)
   > Set lookingDownExtreme(lookingDownExtreme:double=MapRangeClamped(Value=@n3.Yaw, InRangeA=20, InRangeB=34)) -> (Output_Get:double)
   n5 = BreakRotator(InRot=@n2.Rotation)
   > Set head_rot_r(head_rot_r:double=MapRangeClamped(Value=@n5.Pitch, InRangeA=10, InRangeB=40)) -> (Output_Get:double)
   > Set headRotDownLeft(headRotDownLeft:double=float * float(A=@n4, B=MapRangeClamped(Value=@n5.Pitch, InRangeA=0, InRangeB=-10)) {is=title K2Node_PromotableOperator}) -> (Output_Get:double)
   > Set headRotDownRight(headRotDownRight:double=float * float(A=@n4, B=MapRangeClamped(Value=@n5.Pitch, InRangeA=0, InRangeB=15))) -> (Output_Get:double)
   n6 = Cast To FortPlayerPawn(execute=@n1.then_1, Object=TryGetFortPawn())
   > Set body_mesh(body_mesh:Skeletal Mesh Component Object Reference=GetSkeletalMeshForPartType(self=@n6)) -> (Output_Get:Skeletal Mesh Component Object Reference)
   n7 = Get!(execute:Exec=@n1.then_2) -> (then:Exec, else:Exec, body_mesh:Skeletal Mesh Component Object Reference)
   > n8 = Set HasRoyalWonkBody(HasRoyalWonkBody:bool=MatchesWildcard(SourceString=GetDisplayName(Object=@n7), Wildcard=*M_MED_RoyalWonk, SearchCase=CaseSensitive)) -> (Output_Get:bool)
   n9 = Get bRigidBodyEnabled()
   > Set RoyalWonkNoDynamics(RoyalWonkNoDynamics:bool=AND Boolean(A=@n8, B=NOT Boolean(A=@n9))) -> (Output_Get:bool)
   n10 = Get(execute=@n1.then_3)
   n11 = GetSocketBoneName!(self:Skinned Mesh Component Object Reference=@n10, InSocketName:Name=base_bow) -> (ReturnValue:Name=None)
   > n12 = Set isWonkBodyAttached(isWonkBodyAttached:bool="Not Equal Exactly (String)"(A="To String (Name)"(InName=@n11), B=None)) -> (Output_Get:bool)
   > Branch(Condition=@n12)
   n13 = BreakTransform(InTransform=GetSocketTransform(self=Get body_mesh() -> (body_mesh:Skeletal Mesh Component Object Reference), InSocketName=base_bow, TransformSpace=RTS_ParentBoneSpace))
   > Set baseBow_Translation(baseBow_Translation:Vector=@n13) -> (Output_Get:Vector)
   > Set baseBow_Rotation(baseBow_Rotation:Rotator=@n13.Rotation) -> (Output_Get:Rotator)
   n14 = BreakTransform(InTransform=GetSocketTransform(self=Get body_mesh(), InSocketName=dyn_bow, TransformSpace=RTS_ParentBoneSpace))
   > Set dynBow_Translation(dynBow_Translation:Vector=@n14) -> (Output_Get:Vector)
   > Set dynBow_Rotation(dynBow_Rotation:Rotator=@n14.Rotation) -> (Output_Get:Rotator)
  ]}
]
variables [
 baseBow_Rotation {type=Rotator}
 baseBow_Translation {type=Vector}
 body_mesh {type=SkeletalMeshComponent}
 dynBow_Rotation {type=Rotator}
 dynBow_Translation {type=Vector}
 HasRoyalWonkBody {type=boolean}
 head_rot_bk {type=number}
 head_rot_Fr {type=number}
 head_rot_r {type=number}
 headRotDownLeft {type=number}
 headRotDownRight {type=number}
 isWonkBodyAttached {type=boolean}
 lookingDownExtreme {type=number}
 RoyalWonkNoDynamics {type=boolean}
]
